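# Mono MetaSound source for the Loud Kiwi projectile stack sound, using its own attenuation and the ranged weapon projectile effects sound class.
# On play, a short trigger delay fires two wave players in parallel.
# One plays a random stack impact from six variations with randomized upward pitch; the other plays a variable-held wave, pitched up, from a random start time.
# A mixer sums both, with the variable-wave layer at reduced gain.
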MetaSoundSource /LoudKiwi/Assets/Audio/MSS_LoudKiwi_Projectile_Stack
inputs [
 StackImpact {type=WaveAsset:Array}
 UE.Source.OnPlay {type=Trigger}
]
interfaces [UE.OutputFormat.Mono:1.0, UE.Source:1.1]
outputs ["UE.OutputFormat.Mono.Audio:0" {type=Audio}]
pages [Default {
  input_defaults {
   UE.Source.OnPlay=null
   StackImpact [
    /LoudKiwi/Assets/Audio/SoundWavs/LoudKiwi_Proj_Stack_Impact_01
    /LoudKiwi/Assets/Audio/SoundWavs/LoudKiwi_Proj_Stack_Impact_02
    /LoudKiwi/Assets/Audio/SoundWavs/LoudKiwi_Proj_Stack_Impact_03
    /LoudKiwi/Assets/Audio/SoundWavs/LoudKiwi_Proj_Stack_Impact_04
    /LoudKiwi/Assets/Audio/SoundWavs/LoudKiwi_Proj_Stack_Impact_05
    /LoudKiwi/Assets/Audio/SoundWavs/LoudKiwi_Proj_Stack_Impact_06
   ]
  }
  nodes [
   TriggerDelay_0 = UE.Trigger Delay(In=^UE.Source.OnPlay) {also {defaults {Delay Time=0.05}}}
   n1 = UE.RandomTime(Next=@TriggerDelay_0.Out) {also {defaults {Max=9}}}
   WavePlayer_0 = UE.Wave Player:Mono(Play=@n1.On Next, Start Time=@n1.Value, Wave Asset=VariableAccessor.WaveAsset(Variable=InitVariable.WaveAsset().Variable).Value) {also {defaults {Loop="<MetaSoundBuilderResult.SUCCEEDED: 0>", Pitch Shift=6}, version [1,1]}}
   ArrayRandomGet_0 = Array.Random Get:WaveAsset:Array(In Array=^StackImpact, Next=@TriggerDelay_0.Out)
   n2 = UE.RandomFloat(Next=@ArrayRandomGet_0.On Next) {also {defaults {Max=6, Min=4}}}
   WavePlayer_1 = UE.Wave Player:Mono(Wave Asset=@ArrayRandomGet_0.Value, Play=@n2.On Next, Pitch Shift=@n2.Value) {also {version [1,1]}}
   AudioMixerAudioMixerMono2_0 = "AudioMixer.Audio Mixer (Mono, 2)"(In 0=@WavePlayer_0.Out Mono, In 1=@WavePlayer_1.Out Mono) {also {defaults {Gain 0=0.2}}}
  ]
  links [{dst="@UE.OutputFormat.Mono.Audio:0", src=AudioMixerAudioMixerMono2_0.Out}]
 }]
properties {
 assetImportData=&self:AssetImportData
 attenuationSettings=&/ATT_LoudKiwi_Projectile_Stack
 bLooping=true
 duration=10000
 maxDistance=6000
 sampleRate=48000
 soundClassObject=&Fort_WeaponRanged_Projectile_Effects
}
