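# Sound cue for the Max variant of the LoudKiwi projectile explosion, on the large explosives sound class.
# It mixes two layers, each picking randomly from three waves: close explosions with the LoudKiwi stack attenuation, and quieter medium explosions with a Fortnite grenade attenuation.
# The mix passes through a randomized pitch-up Modulator, then an Enveloper with a rising pitch curve.
# A concurrency override caps simultaneous instances.
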SoundCue /LoudKiwi/Assets/Audio/SC_LoudKiwi_Projectile_Explosion_Max
root=Enveloper_0
nodes [
 {id=Attenuation_2, children [Random_4], properties {attenuationSettings=/Game/Sounds/Attenuation/Fort_Explosion_Medium_Athena_Grenade_Att}}
 {id=Attenuation_3, children [Random_5], properties {attenuationSettings=&/ATT_LoudKiwi_Projectile_Stack}}
 {
  id=Enveloper_0
  children [Modulator_0]
  properties {pitchCurve {editorCurveData {keys [{time=0}, <at 1 1.4>]}}, volumeCurve {editorCurveData {keys [<at 0 1>]}}}
 }
 {id=Mixer_3, children [Attenuation_3, Attenuation_2], properties {inputVolume [1,0.75]}}
 {id=Modulator_0, children [Mixer_3], properties {pitchMax=1.3, pitchMin=1.2}}
 {id=Random_4, children [WavePlayer_10, WavePlayer_11, WavePlayer_12], properties {weights [1,1,1]}}
 {id=Random_5, children [WavePlayer_3,WavePlayer_4,WavePlayer_5], properties {weights [1,1,1]}}
 {id=WavePlayer_10, properties {soundWaveAssetPtr=/LoudKiwi/Assets/Audio/SoundWavs/LoudKiwi_Proj_Stack_Explosion_Medium_01}}
 {id=WavePlayer_11, properties {soundWaveAssetPtr=/LoudKiwi/Assets/Audio/SoundWavs/LoudKiwi_Proj_Stack_Explosion_Medium_02}}
 {id=WavePlayer_12, properties {soundWaveAssetPtr=/LoudKiwi/Assets/Audio/SoundWavs/LoudKiwi_Proj_Stack_Explosion_Medium_03}}
 {id=WavePlayer_3, properties {soundWaveAssetPtr=/LoudKiwi/Assets/Audio/SoundWavs/LoudKiwi_Proj_Stack_Explosion_Close_01}}
 {id=WavePlayer_4, properties {soundWaveAssetPtr=/LoudKiwi/Assets/Audio/SoundWavs/LoudKiwi_Proj_Stack_Explosion_Close_02}}
 {id=WavePlayer_5, properties {soundWaveAssetPtr=/LoudKiwi/Assets/Audio/SoundWavs/LoudKiwi_Proj_Stack_Explosion_Close_03}}
]
properties {
 bOverrideConcurrency=true
 maxDistance=9625
 soundClassObject=/Game/Sounds/FortSoundClassesAndModes/Fort_Explosives_Large_PSM
 volumeMultiplier=0.5
 concurrencyOverrides {maxCount=5}
}
