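# Sound cue for the LoudKiwi projectile explosion, named as the non-player variant, in the large explosives sound class with capped concurrency.
# It mixes randomized Cube Queen bullet impacts with randomized medium stack explosions that carry their own medium-range attenuation, separate from the cue's close-range one.
# A quality-level node gives the first two levels five impact variations and the last two three.
# The mix runs through a raised random pitch modulator and a pitch envelope.
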SoundCue /LoudKiwi/Assets/Audio/SC_LoudKiwi_Projectile_Explosion_NonPlayer
root=Enveloper_0
nodes [
 {id=Attenuation_2, children [Random_4], properties {attenuationSettings=/Game/Sounds/Attenuation/Fort_Explosion_Medium_Athena_Grenade_Att}}
 {
  id=Enveloper_0
  children [Modulator_0]
  properties {pitchCurve {editorCurveData {keys [{time=0}, <at 1 1.4>]}}, volumeCurve {editorCurveData {keys [<at 0 1>]}}}
 }
 {id=Mixer_1, children [QualityLevel_0, Attenuation_2], properties {inputVolume [1,1]}}
 {id=Modulator_0, children [Mixer_1], properties {pitchMax=1.3, pitchMin=1.2}}
 {id=QualityLevel_0, children [Random_0,Random_0,Random_1,Random_1]}
 {id=Random_0, children [WavePlayer_0,WavePlayer_1,WavePlayer_2,WavePlayer_6,WavePlayer_7], properties {weights [1,1,1,1,1]}}
 {id=Random_1, children [WavePlayer_8, WavePlayer_9, WavePlayer_13], properties {weights [1,1,1]}}
 {id=Random_4, children [WavePlayer_10, WavePlayer_11, WavePlayer_12], properties {weights [1,1,1]}}
 {id=WavePlayer_0, properties {soundWaveAssetPtr=/LoudKiwi/Assets/Audio/SoundWavs/Cube_Queen_Bullet_Impact_04}}
 {id=WavePlayer_1, properties {soundWaveAssetPtr=/LoudKiwi/Assets/Audio/SoundWavs/Cube_Queen_Bullet_Impact_05}}
 {id=WavePlayer_10, properties {soundWaveAssetPtr=/LoudKiwi/Assets/Audio/SoundWavs/LoudKiwi_Proj_Stack_Explosion_Medium_01}}
 {id=WavePlayer_11, properties {soundWaveAssetPtr=/LoudKiwi/Assets/Audio/SoundWavs/LoudKiwi_Proj_Stack_Explosion_Medium_02}}
 {id=WavePlayer_12, properties {soundWaveAssetPtr=/LoudKiwi/Assets/Audio/SoundWavs/LoudKiwi_Proj_Stack_Explosion_Medium_03}}
 {id=WavePlayer_13, properties {soundWaveAssetPtr=/LoudKiwi/Assets/Audio/SoundWavs/Cube_Queen_Bullet_Impact_08}}
 {id=WavePlayer_2, properties {soundWaveAssetPtr=/LoudKiwi/Assets/Audio/SoundWavs/Cube_Queen_Bullet_Impact_06}}
 {id=WavePlayer_6, properties {soundWaveAssetPtr=/LoudKiwi/Assets/Audio/SoundWavs/Cube_Queen_Bullet_Impact_07}}
 {id=WavePlayer_7, properties {soundWaveAssetPtr=&Cube_Queen_Bullet_Impact_08}}
 {id=WavePlayer_8, properties {soundWaveAssetPtr=&Cube_Queen_Bullet_Impact_06}}
 {id=WavePlayer_9, properties {soundWaveAssetPtr=&Cube_Queen_Bullet_Impact_07}}
]
properties {
 attenuationSettings=/Game/Sounds/Attenuation/Fort_Explosion_Close_Athena_Grenade_Att
 bOverrideConcurrency=true
 maxDistance=9625
 soundClassObject=/Game/Sounds/FortSoundClassesAndModes/Fort_Explosives_Large_PSM
 volumeMultiplier=0.5
 concurrencyOverrides {maxCount=5}
}
