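# Sound cue for a LoudKiwi projectile hitting water, layering water splash and cube bullet-impact one-shots with shared attenuation and limited concurrency.
# A quality switch selects between two mixers: the first two levels use full random pools for both layers, the last two use smaller pools.
# The result passes through random upward pitch modulation, then a root enveloper that sweeps pitch upward over time.
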SoundCue /LoudKiwi/Assets/Audio/SC_LoudKiwi_Projectile_Impact_Water
root=Enveloper_0
nodes [
 {
  id=Enveloper_0
  children [Modulator_0]
  properties {pitchCurve {editorCurveData {keys [<at 0 0.8>,<at 1 2.5>]}}, volumeCurve {editorCurveData {keys [<at 0 1>]}}}
 }
 {id=Mixer_0, children [Random_0,Random_2], properties {inputVolume [1,1]}}
 {id=Mixer_1, children [Random_3,Random_1], properties {inputVolume [1,1]}}
 {id=Modulator_0, children [QualityLevel_0], properties {pitchMax=1.5, pitchMin=1}}
 {id=QualityLevel_0, children [Mixer_0,Mixer_0,Mixer_1,Mixer_1]}
 {
  id=Random_0
  children [WavePlayer_0,WavePlayer_1,WavePlayer_2,WavePlayer_3,WavePlayer_4,WavePlayer_5,WavePlayer_6,WavePlayer_7,WavePlayer_8,WavePlayer_9]
  properties {weights [1,1,1,1,1,1,1,1,1,1]}
 }
 {id=Random_1, children [WavePlayer_10, WavePlayer_11, WavePlayer_12], properties {weights [1,1,1]}}
 {
  id=Random_2
  children [
   WavePlayer_21
   WavePlayer_22
   WavePlayer_23
   WavePlayer_24
   WavePlayer_25
   WavePlayer_26
   WavePlayer_27
   WavePlayer_28
   WavePlayer_29
   WavePlayer_30
  ]
  properties {weights [1,1,1,1,1,1,1,1,1,1]}
 }
 {id=Random_3, children [WavePlayer_31, WavePlayer_32, WavePlayer_33, WavePlayer_34], properties {weights [1,1,1,1]}}
 {id=WavePlayer_0, properties {soundWaveAssetPtr=/LoudKiwi/Assets/Audio/SoundWavs/Bullet_Impact_Water_Med_01}}
 {id=WavePlayer_1, properties {soundWaveAssetPtr=/LoudKiwi/Assets/Audio/SoundWavs/Bullet_Impact_Water_Med_02}}
 {id=WavePlayer_10, properties {soundWaveAssetPtr=&Bullet_Impact_Water_Med_01}}
 {id=WavePlayer_11, properties {soundWaveAssetPtr=&Bullet_Impact_Water_Med_02}}
 {id=WavePlayer_12, properties {soundWaveAssetPtr=/LoudKiwi/Assets/Audio/SoundWavs/Bullet_Impact_Water_Med_03}}
 {id=WavePlayer_2, properties {soundWaveAssetPtr=&Bullet_Impact_Water_Med_03}}
 {id=WavePlayer_21, properties {soundWaveAssetPtr=/LoudKiwi/Assets/Audio/SoundWavs/Cube_Phys_Mat_Bullet_Impact_01}}
 {id=WavePlayer_22, properties {soundWaveAssetPtr=/LoudKiwi/Assets/Audio/SoundWavs/Cube_Phys_Mat_Bullet_Impact_02}}
 {id=WavePlayer_23, properties {soundWaveAssetPtr=/LoudKiwi/Assets/Audio/SoundWavs/Cube_Phys_Mat_Bullet_Impact_03}}
 {id=WavePlayer_24, properties {soundWaveAssetPtr=/LoudKiwi/Assets/Audio/SoundWavs/Cube_Phys_Mat_Bullet_Impact_04}}
 {id=WavePlayer_25, properties {soundWaveAssetPtr=/LoudKiwi/Assets/Audio/SoundWavs/Cube_Phys_Mat_Bullet_Impact_05}}
 {id=WavePlayer_26, properties {soundWaveAssetPtr=/LoudKiwi/Assets/Audio/SoundWavs/Cube_Phys_Mat_Bullet_Impact_06}}
 {id=WavePlayer_27, properties {soundWaveAssetPtr=/LoudKiwi/Assets/Audio/SoundWavs/Cube_Phys_Mat_Bullet_Impact_07}}
 {id=WavePlayer_28, properties {soundWaveAssetPtr=/LoudKiwi/Assets/Audio/SoundWavs/Cube_Phys_Mat_Bullet_Impact_08}}
 {id=WavePlayer_29, properties {soundWaveAssetPtr=/LoudKiwi/Assets/Audio/SoundWavs/Cube_Phys_Mat_Bullet_Impact_09}}
 {id=WavePlayer_3, properties {soundWaveAssetPtr=/LoudKiwi/Assets/Audio/SoundWavs/Bullet_Impact_Water_Med_04}}
 {id=WavePlayer_30, properties {soundWaveAssetPtr=/LoudKiwi/Assets/Audio/SoundWavs/Cube_Phys_Mat_Bullet_Impact_10}}
 {id=WavePlayer_31, properties {soundWaveAssetPtr=&Cube_Phys_Mat_Bullet_Impact_01}}
 {id=WavePlayer_32, properties {soundWaveAssetPtr=&Cube_Phys_Mat_Bullet_Impact_02}}
 {id=WavePlayer_33, properties {soundWaveAssetPtr=&Cube_Phys_Mat_Bullet_Impact_03}}
 {id=WavePlayer_34, properties {soundWaveAssetPtr=&Cube_Phys_Mat_Bullet_Impact_04}}
 {id=WavePlayer_4, properties {soundWaveAssetPtr=/LoudKiwi/Assets/Audio/SoundWavs/Bullet_Impact_Water_Med_05}}
 {id=WavePlayer_5, properties {soundWaveAssetPtr=/LoudKiwi/Assets/Audio/SoundWavs/Bullet_Impact_Water_Med_06}}
 {id=WavePlayer_6, properties {soundWaveAssetPtr=/LoudKiwi/Assets/Audio/SoundWavs/Bullet_Impact_Water_Med_07}}
 {id=WavePlayer_7, properties {soundWaveAssetPtr=/LoudKiwi/Assets/Audio/SoundWavs/Bullet_Impact_Water_Med_08}}
 {id=WavePlayer_8, properties {soundWaveAssetPtr=/LoudKiwi/Assets/Audio/SoundWavs/Bullet_Impact_Water_Med_09}}
 {id=WavePlayer_9, properties {soundWaveAssetPtr=/LoudKiwi/Assets/Audio/SoundWavs/Bullet_Impact_Water_Med_10}}
]
properties {
 attenuationSettings=/Game/Sounds/Attenuation/Fort_Bullet_Impact_Att
 bOverrideConcurrency=true
 maxDistance=5300
 soundClassObject=/Game/Sounds/FortSoundClassesAndModes/Fort_Explosives_Large_PSM
 volumeMultiplier=0.2
 concurrencyOverrides {maxCount=5}
}
