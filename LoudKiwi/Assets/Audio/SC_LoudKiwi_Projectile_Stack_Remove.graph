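# Sound cue for removing a stack from the LoudKiwi projectile, classed as a ranged weapon projectile effect.
# A Random node picks one of three equally weighted removal waves.
# A Modulator applies a fixed upward pitch shift, then a root Enveloper varies pitch over time with a flat volume curve.
# The cue plays at reduced volume and references a dedicated attenuation asset for distance falloff.
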SoundCue /LoudKiwi/Assets/Audio/SC_LoudKiwi_Projectile_Stack_Remove
root=Enveloper_0
nodes [
 {
  id=Enveloper_0
  children [Modulator_0]
  properties {pitchCurve {editorCurveData {keys [{time=0}, <at 1 0.8>]}}, volumeCurve {editorCurveData {keys [<at 0 1>]}}}
 }
 {id=Modulator_0, children [Random_4], properties {pitchMax=1.5, pitchMin=1.5}}
 {id=Random_4, children [WavePlayer_16, WavePlayer_17, WavePlayer_18], properties {weights [1,1,1]}}
 {id=WavePlayer_16, properties {soundWaveAssetPtr=/LoudKiwi/Assets/Audio/SoundWavs/LoudKiwi_Proj_Stack_Remove_01}}
 {id=WavePlayer_17, properties {soundWaveAssetPtr=/LoudKiwi/Assets/Audio/SoundWavs/LoudKiwi_Proj_Stack_Remove_02}}
 {id=WavePlayer_18, properties {soundWaveAssetPtr=/LoudKiwi/Assets/Audio/SoundWavs/LoudKiwi_Proj_Stack_Remove_03}}
]
properties {
 attenuationSettings=&/ATT_LoudKiwi_Projectile_Stack_Remove
 maxDistance=4000
 soundClassObject=&Fort_WeaponRanged_Projectile_Effects
 volumeMultiplier=0.15
 attenuationOverrides {
  bAttenuateWithLPF=true
  distanceAlgorithm=Logarithmic
  falloffDistance=2500
  lPFFrequencyAtMax=2000
  lPFRadiusMax=3000
  lPFRadiusMin=500
  attenuationShapeExtents {x=500}
 }
}
