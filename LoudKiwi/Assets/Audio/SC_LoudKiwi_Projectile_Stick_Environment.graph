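# Sound cue for the LoudKiwi projectile sticking to the environment, using bullet impact attenuation with a concurrency override.
# A quality switch sends its first two levels to a two-layer mix and its last two to a single enveloped random cube impact.
# The mix layers a delayed, enveloped random cube impact with a random pistol bullet impact.
# Nested modulators randomize pitch upward, and envelopers apply a short pitch curve to the cube impacts.
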SoundCue /LoudKiwi/Assets/Audio/SC_LoudKiwi_Projectile_Stick_Environment
root=Modulator_0
nodes [
 {id=Delay_0, children [Random_2], properties {delayMax=0.1, delayMin=0.1}}
 {
  id=Enveloper_0
  children [Modulator_2]
  properties {pitchCurve {editorCurveData {keys [{time=0}, <at 0.3 0.5>]}}, volumeCurve {editorCurveData {keys [<at 0 1>]}}}
 }
 {
  id=Enveloper_1
  children [Modulator_3]
  properties {pitchCurve {editorCurveData {keys [{time=0}, <at 0.3 0.5>]}}, volumeCurve {editorCurveData {keys [<at 0 1>]}}}
 }
 {id=Mixer_0, children [Enveloper_0,Modulator_1], properties {inputVolume [1,0.8]}}
 {id=Modulator_0, children [QualityLevel_0], properties {pitchMax=1.3, pitchMin=1.2}}
 {id=Modulator_1, children [Random_0], properties {pitchMax=1.5, pitchMin=1.3}}
 {id=Modulator_2, children [Delay_0], properties {pitchMax=1.3, pitchMin=1.2}}
 {id=Modulator_3, children [Random_3], properties {pitchMax=1.3, pitchMin=1.2}}
 {id=QualityLevel_0, children [Mixer_0,Mixer_0,Enveloper_1,Enveloper_1]}
 {id=Random_0, children [WavePlayer_0,WavePlayer_1,WavePlayer_2,WavePlayer_3], properties {weights [1,1,1,1]}}
 {
  id=Random_2
  children [
   WavePlayer_21
   WavePlayer_22
   WavePlayer_23
   WavePlayer_24
   WavePlayer_25
   WavePlayer_26
   WavePlayer_27
   WavePlayer_28
   WavePlayer_29
   WavePlayer_30
  ]
  properties {weights [1,1,1,1,1,1,1,1,1,1]}
 }
 {id=Random_3, children [WavePlayer_31, WavePlayer_32, WavePlayer_33, WavePlayer_34], properties {weights [1,1,1,1]}}
 {id=WavePlayer_0, properties {soundWaveAssetPtr=/LoudKiwi/Assets/Audio/SoundWavs/player_bulletimpact_pistol_01}}
 {id=WavePlayer_1, properties {soundWaveAssetPtr=/LoudKiwi/Assets/Audio/SoundWavs/player_bulletimpact_pistol_02}}
 {id=WavePlayer_2, properties {soundWaveAssetPtr=/LoudKiwi/Assets/Audio/SoundWavs/player_bulletimpact_pistol_03}}
 {id=WavePlayer_21, properties {soundWaveAssetPtr=/LoudKiwi/Assets/Audio/SoundWavs/Cube_Phys_Mat_Bullet_Impact_01}}
 {id=WavePlayer_22, properties {soundWaveAssetPtr=/LoudKiwi/Assets/Audio/SoundWavs/Cube_Phys_Mat_Bullet_Impact_02}}
 {id=WavePlayer_23, properties {soundWaveAssetPtr=/LoudKiwi/Assets/Audio/SoundWavs/Cube_Phys_Mat_Bullet_Impact_03}}
 {id=WavePlayer_24, properties {soundWaveAssetPtr=/LoudKiwi/Assets/Audio/SoundWavs/Cube_Phys_Mat_Bullet_Impact_04}}
 {id=WavePlayer_25, properties {soundWaveAssetPtr=/LoudKiwi/Assets/Audio/SoundWavs/Cube_Phys_Mat_Bullet_Impact_05}}
 {id=WavePlayer_26, properties {soundWaveAssetPtr=/LoudKiwi/Assets/Audio/SoundWavs/Cube_Phys_Mat_Bullet_Impact_06}}
 {id=WavePlayer_27, properties {soundWaveAssetPtr=/LoudKiwi/Assets/Audio/SoundWavs/Cube_Phys_Mat_Bullet_Impact_07}}
 {id=WavePlayer_28, properties {soundWaveAssetPtr=/LoudKiwi/Assets/Audio/SoundWavs/Cube_Phys_Mat_Bullet_Impact_08}}
 {id=WavePlayer_29, properties {soundWaveAssetPtr=/LoudKiwi/Assets/Audio/SoundWavs/Cube_Phys_Mat_Bullet_Impact_09}}
 {id=WavePlayer_3, properties {soundWaveAssetPtr=/LoudKiwi/Assets/Audio/SoundWavs/player_bulletimpact_pistol_04}}
 {id=WavePlayer_30, properties {soundWaveAssetPtr=/LoudKiwi/Assets/Audio/SoundWavs/Cube_Phys_Mat_Bullet_Impact_10}}
 {id=WavePlayer_31, properties {soundWaveAssetPtr=&Cube_Phys_Mat_Bullet_Impact_01}}
 {id=WavePlayer_32, properties {soundWaveAssetPtr=&Cube_Phys_Mat_Bullet_Impact_02}}
 {id=WavePlayer_33, properties {soundWaveAssetPtr=&Cube_Phys_Mat_Bullet_Impact_03}}
 {id=WavePlayer_34, properties {soundWaveAssetPtr=&Cube_Phys_Mat_Bullet_Impact_04}}
]
properties {
 attenuationSettings=/Game/Sounds/Attenuation/Fort_Bullet_Impact_Att
 bOverrideConcurrency=true
 maxDistance=5300
 soundClassObject=/Game/Sounds/FortSoundClassesAndModes/Fort_Explosives_Large_PSM
 volumeMultiplier=0.15
 concurrencyOverrides {maxCount=3}
}
