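# Sound cue for a LoudKiwi projectile sticking to a player, using bullet-impact attenuation and a three-voice concurrency cap.
# A quality-level root selects one of two mixes, each layering a random pistol impact over a quieter, slightly delayed Cube Queen impact.
# The delayed layer passes through random pitch modulation and a rising-pitch envelope.
# The first two quality slots use more variations and lower pitch; the last two use fewer, pitched higher.
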SoundCue /LoudKiwi/Assets/Audio/SC_LoudKiwi_Projectile_Stick_Player
root=QualityLevel_0
nodes [
 {id=Delay_0, children [Enveloper_1], properties {delayMax=0.1, delayMin=0.1}}
 {id=Delay_1, children [Enveloper_2], properties {delayMax=0.1, delayMin=0.1}}
 {
  id=Enveloper_1
  children [Modulator_1]
  properties {pitchCurve {editorCurveData {keys [{time=0}, <at 1 1.2>]}}, volumeCurve {editorCurveData {keys [<at 0 1>]}}}
 }
 {
  id=Enveloper_2
  children [Modulator_0]
  properties {pitchCurve {editorCurveData {keys [{time=0}, <at 1 1.2>]}}, volumeCurve {editorCurveData {keys [<at 0 1>]}}}
 }
 {id=Mixer_0, children [Delay_0,Random_6], properties {inputVolume [0.4,1]}}
 {id=Mixer_2, children [Delay_1,Random_8], properties {inputVolume [0.4,1]}}
 {id=Modulator_0, children [Random_7], properties {pitchMax=1.5, pitchMin=1.4}}
 {id=Modulator_1, children [Random_5], properties {pitchMax=0.8, pitchMin=0.7}}
 {id=QualityLevel_0, children [Mixer_0,Mixer_0,Mixer_2,Mixer_2]}
 {
  id=Random_5
  children [WavePlayer_38, WavePlayer_39, WavePlayer_40, WavePlayer_42, WavePlayer_43, WavePlayer_44, WavePlayer_45, WavePlayer_46]
  properties {weights [1,1,1,1,1,1,1,1]}
 }
 {id=Random_6, children [WavePlayer_34, WavePlayer_35, WavePlayer_36, WavePlayer_37], properties {weights [1,1,1,1]}}
 {id=Random_7, children [WavePlayer_55, WavePlayer_56, WavePlayer_57], properties {weights [1,1,1]}}
 {id=Random_8, children [WavePlayer_51, WavePlayer_52, WavePlayer_53], properties {weights [1,1,1]}}
 {id=WavePlayer_34, properties {soundWaveAssetPtr=/LoudKiwi/Assets/Audio/SoundWavs/player_bulletimpact_pistol_01}}
 {id=WavePlayer_35, properties {soundWaveAssetPtr=/LoudKiwi/Assets/Audio/SoundWavs/player_bulletimpact_pistol_02}}
 {id=WavePlayer_36, properties {soundWaveAssetPtr=/LoudKiwi/Assets/Audio/SoundWavs/player_bulletimpact_pistol_03}}
 {id=WavePlayer_37, properties {soundWaveAssetPtr=/LoudKiwi/Assets/Audio/SoundWavs/player_bulletimpact_pistol_04}}
 {id=WavePlayer_38, properties {soundWaveAssetPtr=/LoudKiwi/Assets/Audio/SoundWavs/Cube_Queen_Bullet_Impact_01}}
 {id=WavePlayer_39, properties {soundWaveAssetPtr=/LoudKiwi/Assets/Audio/SoundWavs/Cube_Queen_Bullet_Impact_02}}
 {id=WavePlayer_40, properties {soundWaveAssetPtr=/LoudKiwi/Assets/Audio/SoundWavs/Cube_Queen_Bullet_Impact_03}}
 {id=WavePlayer_42, properties {soundWaveAssetPtr=/LoudKiwi/Assets/Audio/SoundWavs/Cube_Queen_Bullet_Impact_05}}
 {id=WavePlayer_43, properties {soundWaveAssetPtr=/LoudKiwi/Assets/Audio/SoundWavs/Cube_Queen_Bullet_Impact_06}}
 {id=WavePlayer_44, properties {soundWaveAssetPtr=/LoudKiwi/Assets/Audio/SoundWavs/Cube_Queen_Bullet_Impact_07}}
 {id=WavePlayer_45, properties {soundWaveAssetPtr=/LoudKiwi/Assets/Audio/SoundWavs/Cube_Queen_Bullet_Impact_08}}
 {id=WavePlayer_46, properties {soundWaveAssetPtr=/LoudKiwi/Assets/Audio/SoundWavs/Cube_Queen_Bullet_Impact_09}}
 {id=WavePlayer_51, properties {soundWaveAssetPtr=&player_bulletimpact_pistol_01}}
 {id=WavePlayer_52, properties {soundWaveAssetPtr=&player_bulletimpact_pistol_02}}
 {id=WavePlayer_53, properties {soundWaveAssetPtr=&player_bulletimpact_pistol_03}}
 {id=WavePlayer_55, properties {soundWaveAssetPtr=&Cube_Queen_Bullet_Impact_01}}
 {id=WavePlayer_56, properties {soundWaveAssetPtr=&Cube_Queen_Bullet_Impact_02}}
 {id=WavePlayer_57, properties {soundWaveAssetPtr=&Cube_Queen_Bullet_Impact_03}}
]
properties {
 attenuationSettings=/Game/Sounds/Attenuation/Fort_Bullet_Impact_Att
 bOverrideConcurrency=true
 maxDistance=5300
 soundClassObject=/Game/Sounds/FortSoundClassesAndModes/Fort_Explosives_Large_PSM
 volumeMultiplier=0.2
 concurrencyOverrides {maxCount=3}
}
