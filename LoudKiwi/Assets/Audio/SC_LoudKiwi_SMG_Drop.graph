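# Sound cue for the LoudKiwi SMG drop sound, played as a weapon drop effect.
# A Random node picks one of three equally weighted drop recordings and feeds a Modulator that applies a slightly varied volume boost.
# The cue plays at lowered pitch, uses shared weapon-drop attenuation and a stop-oldest concurrency set, and is assigned to the loot-collect sound class.
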SoundCue /LoudKiwi/Assets/Audio/SC_LoudKiwi_SMG_Drop
root=Modulator_2
nodes [
 {id=Modulator_2, children [Random_3], properties {volumeMax=1.3, volumeMin=1.25}}
 {id=Random_3, children [WavePlayer_0,WavePlayer_1,WavePlayer_2], properties {weights [1,1,1]}}
 {id=WavePlayer_0, properties {soundWaveAssetPtr=/LoudKiwi/Assets/Audio/SoundWavs/LoudKiwi_SMG_Drop_01}}
 {id=WavePlayer_1, properties {soundWaveAssetPtr=/LoudKiwi/Assets/Audio/SoundWavs/LoudKiwi_SMG_Drop_02}}
 {id=WavePlayer_2, properties {soundWaveAssetPtr=/LoudKiwi/Assets/Audio/SoundWavs/LoudKiwi_SMG_Drop_03}}
]
properties {
 attenuationSettings=&Fort_Weapon_Drop_Att
 maxDistance=3000
 pitchMultiplier=0.9
 soundClassObject=&Fort_Loot_Collect
 volumeMultiplier=1
 concurrencyOverrides {resolutionRule=PreventNew}
 concurrencySet [&Weapon_Drop_StopOldest]
}
