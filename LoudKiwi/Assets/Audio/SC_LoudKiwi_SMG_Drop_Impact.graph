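# Sound cue for the LoudKiwi SMG drop impact, played when the weapon lands.
# A Random node picks one of three equally weighted landing waves and feeds it through a Modulator at the root.
# It uses weapon-drop attenuation and the loot-collect sound class.
# A concurrency override allows only a few simultaneous instances and prevents new ones beyond that limit.
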SoundCue /LoudKiwi/Assets/Audio/SC_LoudKiwi_SMG_Drop_Impact
root=Modulator_2
nodes [
 {id=Modulator_2, children [Random_3]}
 {id=Random_3, children [WavePlayer_1,WavePlayer_2,WavePlayer_3], properties {weights [1,1,1]}}
 {id=WavePlayer_1, properties {soundWaveAssetPtr=/LoudKiwi/Assets/Audio/SoundWavs/LoudKiwi_SMG_Land_03}}
 {id=WavePlayer_2, properties {soundWaveAssetPtr=/LoudKiwi/Assets/Audio/SoundWavs/LoudKiwi_SMG_Land_02}}
 {id=WavePlayer_3, properties {soundWaveAssetPtr=/LoudKiwi/Assets/Audio/SoundWavs/LoudKiwi_SMG_Land_01}}
]
properties {
 attenuationSettings=&Fort_Weapon_Drop_Att
 bOverrideConcurrency=true
 maxDistance=3000
 soundClassObject=&Fort_Loot_Collect
 volumeMultiplier=1
 attenuationOverrides {falloffDistance=1000, attenuationShapeExtents {x=0}}
 concurrencyOverrides {maxCount=3, resolutionRule=PreventNew}
}
