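# Sound cue for the LoudKiwi SMG out-of-ammo sound, reusing rifle out-of-ammo waves at raised pitch and reduced volume.
# A Modulator with pitch variation disabled feeds a Quality Level node that picks an equal-weight Random pool.
# The first two quality slots share four variations, while the last two share only the first two.
# It uses close weapon attenuation, a first-person ranged-fire sound class and first-person gunfire concurrency.
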SoundCue /LoudKiwi/Assets/Audio/SC_LoudKiwi_SMG_OutofAmmo
root=Modulator_0
nodes [
 {id=Modulator_0, children [QualityLevel_0], properties {pitchMax=1, pitchMin=1}}
 {id=QualityLevel_0, children [Random_0,Random_0,Random_1,Random_1]}
 {id=Random_0, children [WavePlayer_4,WavePlayer_5,WavePlayer_6,WavePlayer_7], properties {weights [1,1,1,1]}}
 {id=Random_1, children [WavePlayer_8,WavePlayer_9], properties {weights [1,1]}}
 {id=WavePlayer_4, properties {soundWaveAssetPtr=/LoudKiwi/Assets/Audio/SoundWavs/Rifle_OutofAmmo_01}}
 {id=WavePlayer_5, properties {soundWaveAssetPtr=/LoudKiwi/Assets/Audio/SoundWavs/Rifle_OutofAmmo_02}}
 {id=WavePlayer_6, properties {soundWaveAssetPtr=/LoudKiwi/Assets/Audio/SoundWavs/Rifle_OutofAmmo_03}}
 {id=WavePlayer_7, properties {soundWaveAssetPtr=/LoudKiwi/Assets/Audio/SoundWavs/Rifle_OutofAmmo_04}}
 {id=WavePlayer_8, properties {soundWaveAssetPtr=&Rifle_OutofAmmo_01}}
 {id=WavePlayer_9, properties {soundWaveAssetPtr=&Rifle_OutofAmmo_02}}
]
properties {
 attenuationSettings=&Fort_Weapon_Projectile_Close_Athena_Att
 maxDistance=4950
 pitchMultiplier=1.2
 soundClassObject=&Fort_WeaponRanged_Fire_1P_PSM
 volumeMultiplier=0.5
 attenuationOverrides {falloffDistance=1600}
 concurrencyOverrides {maxCount=2, resolutionRule=StopOldest}
 concurrencySet [&Weapon_Gunfire_FirstPerson_Conc]
}
