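# Sound cue for the LoudKiwi SMG pickup sound.
# A Random node picks one of three equally weighted pickup waves and feeds it through a Modulator to the output at reduced volume.
# It uses weapon pickup attenuation and the loot collect sound class, references a shared pickup concurrency set, and carries single-instance prevent-new concurrency overrides.
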SoundCue /LoudKiwi/Assets/Audio/SC_LoudKiwi_SMG_Pickup
root=Modulator_2
nodes [
 {id=Modulator_2, children [Random_3]}
 {id=Random_3, children [WavePlayer_0,WavePlayer_1,WavePlayer_8], properties {weights [1,1,1]}}
 {id=WavePlayer_0, properties {soundWaveAssetPtr=/LoudKiwi/Assets/Audio/SoundWavs/LoudKiwi_SMG_Pickup_01}}
 {id=WavePlayer_1, properties {soundWaveAssetPtr=/LoudKiwi/Assets/Audio/SoundWavs/LoudKiwi_SMG_Pickup_02}}
 {id=WavePlayer_8, properties {soundWaveAssetPtr=/LoudKiwi/Assets/Audio/SoundWavs/LoudKiwi_SMG_Pickup_03}}
]
properties {
 attenuationSettings=&Fort_Weapon_Pickup_Att
 maxDistance=3500
 soundClassObject=&Fort_Loot_Collect
 volumeMultiplier=0.25
 concurrencyOverrides {maxCount=1, resolutionRule=PreventNew}
 concurrencySet [&PickupMaximum_Oldest]
}
