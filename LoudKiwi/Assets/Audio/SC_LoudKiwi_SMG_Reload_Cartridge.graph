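# Sound cue for the LoudKiwi SMG reload that plays a cartridge eject foley sound.
# A Random node picks one of three equally weighted eject wave variations and feeds it through a Modulator at the root.
# It uses weapon foley attenuation and sound class, runs at reduced volume, and overrides concurrency.
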SoundCue /LoudKiwi/Assets/Audio/SC_LoudKiwi_SMG_Reload_Cartridge
root=Modulator_1
nodes [
 {id=Modulator_1, children [Random_3]}
 {id=Random_3, children [WavePlayer_0,WavePlayer_1,WavePlayer_2], properties {weights [1,1,1]}}
 {id=WavePlayer_0, properties {soundWaveAssetPtr=/LoudKiwi/Assets/Audio/SoundWavs/LoudKiwi_SMG_Reload_Cartridge_Eject_01}}
 {id=WavePlayer_1, properties {soundWaveAssetPtr=/LoudKiwi/Assets/Audio/SoundWavs/LoudKiwi_SMG_Reload_Cartridge_Eject_02}}
 {id=WavePlayer_2, properties {soundWaveAssetPtr=/LoudKiwi/Assets/Audio/SoundWavs/LoudKiwi_SMG_Reload_Cartridge_Eject_03}}
]
properties {
 attenuationSettings=&AS_Fort_WeaponFoley
 bOverrideConcurrency=true
 maxDistance=2250
 soundClassObject=&Fort_WeaponFoley
 volumeMultiplier=0.35
}
