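# Sound cue for the cartridge insert step of the LoudKiwi SMG reload.
# A random node picks among three equally weighted wave variations and feeds them through a modulator at the root.
# It uses the Fortnite weapon foley sound class and attenuation, runs at reduced volume and enables a concurrency override.
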SoundCue /LoudKiwi/Assets/Audio/SC_LoudKiwi_SMG_Reload_Cartridge_Insert
root=Modulator_1
nodes [
 {id=Modulator_1, children [Random_3]}
 {id=Random_3, children [WavePlayer_8, WavePlayer_9, WavePlayer_10], properties {weights [1,1,1]}}
 {id=WavePlayer_10, properties {soundWaveAssetPtr=/LoudKiwi/Assets/Audio/SoundWavs/LoudKiwi_SMG_Reload_Cartridge_Insert_03}}
 {id=WavePlayer_8, properties {soundWaveAssetPtr=/LoudKiwi/Assets/Audio/SoundWavs/LoudKiwi_SMG_Reload_Cartridge_Insert_01}}
 {id=WavePlayer_9, properties {soundWaveAssetPtr=/LoudKiwi/Assets/Audio/SoundWavs/LoudKiwi_SMG_Reload_Cartridge_Insert_02}}
]
properties {
 attenuationSettings=&AS_Fort_WeaponFoley
 bOverrideConcurrency=true
 maxDistance=2250
 soundClassObject=&Fort_WeaponFoley
 volumeMultiplier=0.401
}
