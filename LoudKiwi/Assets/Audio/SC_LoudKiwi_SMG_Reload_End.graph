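# Sound cue for the end of the LoudKiwi SMG reload, played as weapon foley.
# A Random node picks among three equally weighted reload-end waves and feeds a Modulator at the root.
# The cue uses Fort weapon foley attenuation and sound class, overrides concurrency, raises pitch, lowers volume and has debug enabled.
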SoundCue /LoudKiwi/Assets/Audio/SC_LoudKiwi_SMG_Reload_End
root=Modulator_2
nodes [
 {id=Modulator_2, children [Random_0]}
 {id=Random_0, children [WavePlayer_1,WavePlayer_2,WavePlayer_3], properties {weights [1,1,1]}}
 {id=WavePlayer_1, properties {soundWaveAssetPtr=/LoudKiwi/Assets/Audio/SoundWavs/LoudKiwi_SMG_Reload_End_01}}
 {id=WavePlayer_2, properties {soundWaveAssetPtr=/LoudKiwi/Assets/Audio/SoundWavs/LoudKiwi_SMG_Reload_End_02}}
 {id=WavePlayer_3, properties {soundWaveAssetPtr=/LoudKiwi/Assets/Audio/SoundWavs/LoudKiwi_SMG_Reload_End_03}}
]
properties {
 attenuationSettings=&AS_Fort_WeaponFoley
 bDebug=true
 bOverrideConcurrency=true
 maxDistance=2250
 pitchMultiplier=1.2
 soundClassObject=&Fort_WeaponFoley
 volumeMultiplier=0.6
}
