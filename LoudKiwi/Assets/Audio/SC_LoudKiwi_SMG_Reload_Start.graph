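# Sound cue for the start of the LoudKiwi SMG reload foley.
# A Random node picks one of three equally weighted wave variations and feeds a root Modulator left at default settings for pitch and volume variation.
# It uses the Fortnite weapon foley sound class and attenuation, overrides concurrency locally, and has the debug flag enabled.
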SoundCue /LoudKiwi/Assets/Audio/SC_LoudKiwi_SMG_Reload_Start
root=Modulator_0
nodes [
 {id=Modulator_0, children [Random_0]}
 {id=Random_0, children [WavePlayer_0,WavePlayer_1,WavePlayer_2], properties {weights [1,1,1]}}
 {id=WavePlayer_0, properties {soundWaveAssetPtr=/LoudKiwi/Assets/Audio/SoundWavs/LoudKiwi_SMG_Reload_Start_01}}
 {id=WavePlayer_1, properties {soundWaveAssetPtr=/LoudKiwi/Assets/Audio/SoundWavs/LoudKiwi_SMG_Reload_Start_02}}
 {id=WavePlayer_2, properties {soundWaveAssetPtr=/LoudKiwi/Assets/Audio/SoundWavs/LoudKiwi_SMG_Reload_Start_03}}
]
properties {
 attenuationSettings=&AS_Fort_WeaponFoley
 bDebug=true
 bOverrideConcurrency=true
 maxDistance=2250
 soundClassObject=&Fort_WeaponFoley
 volumeMultiplier=1
}
