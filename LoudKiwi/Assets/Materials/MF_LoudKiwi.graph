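# Material function that layers the animated purple Loud Kiwi effect, apparently a weapon skin, over input material attributes behind a static switch.
# Warped, parallax panning noise, accent textures, banded Fresnel gradients and a box-masked shell glow supply color and glow detail.
# Emissive blends cold, hot, on-fire and reload palettes, with Blueprint-facing power level and fire/reload timestamps providing the animation drivers.
# A channel-selected mask controls the final attribute blend.
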MaterialFunction /LoudKiwi/Assets/Materials/MF_LoudKiwi
graph {nodes [
  Constant()
  n1 = FunctionInput() {inputName=Material, inputType=FunctionInput_MaterialAttributes}
  n2 = NamedRerouteUsage()
  n3 = Saturate(None=DotProduct(
    A=NamedRerouteUsage()
    B=VectorParameter() {group=Loud Kiwi, parameterName=LoudKiwi.MaskChannel, sortPriority=5, defaultValue=<rgba 1 0 0 1>}.RGB
   ))
  n4 = Max(A=@n3, B=NamedRerouteUsage())
  FunctionOutput(None=StaticSwitchParameter(True=BlendMaterialAttributes(A=@n1, B=@n2, Alpha=@n4), False=@n1) {defaultValue=true, group=Loud Kiwi, parameterName=EnableLoudKiwi, sortPriority=1}) {outputName=Material}
  n5 = VectorParameter() {group=Loud Kiwi, parameterName=LoudKiwi.Accent.OnFire.Emissive, sortPriority=5, defaultValue=<rgba 0.9387 0.4793 0.9047 1>}
  n6 = VectorParameter() {group=Loud Kiwi, parameterName=LoudKiwi.Accent.Hot.Emissive, sortPriority=1, defaultValue=<rgba 0.5029 0.0296 0.5583 1>}
  LinearInterpolate(A=@n6.RGB, B=@n5.RGB, Alpha=NamedRerouteUsage())
  n7 = VectorParameter() {group=Loud Kiwi, parameterName=LoudKiwi.Base.Reload.Emissive, sortPriority=5, defaultValue=<rgba 0.9007 0.0529 1 1>}
  n8 = VectorParameter() {group=Loud Kiwi, parameterName=LoudKiwi.Accent.Reload.Emissive, sortPriority=5, defaultValue=<rgba 0.9387 0.4793 0.9047 1>}
  LinearInterpolate(A=@n7.RGB, B=@n8.RGB, Alpha=NamedRerouteUsage())
  n9 = NamedRerouteUsage()
  n10 = Constant()
  n11 = VectorParameter() {group=Loud Kiwi, parameterName=LoudKiwi.Base.Cool.Emissive, sortPriority=1, defaultValue=<rgba 0.1413 0.0075 0.2747 1>}
  n12 = VectorParameter() {group=Loud Kiwi, parameterName=LoudKiwi.Base.Hot.Emissive, sortPriority=1, defaultValue=<rgba 0.9007 0.0529 1 1>}
  n13 = LinearInterpolate(A=@n12.RGB, B=Multiply(A=@n12.RGB, B=NamedRerouteUsage())) {constAlpha=0.7}
  n14 = MaterialFunctionCall(
   Transition Phase=MaterialFunctionCall(Input=NamedRerouteUsage(), Input Low=Constant(), Input High=Constant() {r=1.3}) {materialFunction=&RemapValue_0-1}.Result
   Height Texture=ComponentMask(None=NamedRerouteUsage()) {r=true}
  ) {materialFunction=&HeightLerp}
  n15 = VectorParameter() {group=Loud Kiwi, parameterName=LoudKiwi.Accent.Cold.Emissive, sortPriority=1, defaultValue=<rgba 0.1413 0.0075 0.2747 1>}
  n16 = LinearInterpolate(A=@n15.RGB, B=@n6.RGB, Alpha=NamedRerouteUsage())
  n17 = Power(Base=NamedRerouteUsage()) {constExponent=0.5}
  n18 = VectorParameter() {
   group=Loud Kiwi
   parameterName=LoudKiwi.CenterGlow.Cold.Emissive
   sortPriority=50
   channelNames {a=NoiseBlend, b="", g="", r=""}
   defaultValue=<rgba 0.692 0.098 0.8177 1>
  }
  n19 = NamedRerouteUsage()
  n20 = Multiply(A=LinearInterpolate(Alpha=@n18.NoiseBlend) {constA=-1}, B=@n19) {constB=0.01}
  n21 = VectorParameter() {
   group=Loud Kiwi
   parameterName=LoudKiwi.CenterGlow.Hot.Emissive
   sortPriority=50
   channelNames {a=NoiseBlend, b="", g="", r=""}
   defaultValue=<rgba 0.692 0.098 0.8177 1>
  }
  n22 = Multiply(A=LinearInterpolate(Alpha=@n21.NoiseBlend) {constA=-1}, B=@n19) {constB=0.01}
  n23 = Multiply(
   A=Saturate(None=Add(A=@n17) {constB=0.3})
   B=LinearInterpolate(A=Add(A=@n18.RGB, B=@n20), B=Add(A=@n21.RGB, B=@n22), Alpha=NamedRerouteUsage())
  )
  n24 = NamedRerouteUsage()
  n25 = VectorParameter() {group=Loud Kiwi, parameterName=LoudKiwi.Base.OnFire.Emissive, sortPriority=5, defaultValue=<rgba 0.9387 0.4793 0.9047 1>}
  n26 = VectorParameter() {
   group=Loud Kiwi
   parameterName=LoudKiwi.CenterGlow.OnFire.Emissive
   sortPriority=50
   channelNames {a=NoiseBlend, b="", g="", r=""}
   defaultValue=<rgba 0.692 0.098 0.8177 1>
  }
  n27 = Multiply(A=LinearInterpolate(Alpha=@n26.NoiseBlend) {constA=-1}, B=NamedRerouteUsage()) {constB=0.01}
  n28 = LinearInterpolate(A=@n25.RGB, B=LinearInterpolate(A=@n5.RGB, B=Add(A=@n26.RGB, B=@n27), Alpha=NamedRerouteUsage()), Alpha=NamedRerouteUsage())
  n29 = LinearInterpolate(
   A=LinearInterpolate(A=LinearInterpolate(A=@n11.RGB, B=@n13, Alpha=@n14.Alpha), B=LinearInterpolate(A=@n16, B=@n23, Alpha=@n24), Alpha=NamedRerouteUsage())
   B=@n28
   Alpha=NamedRerouteUsage()
  ) {constA=1, constB=5}
  n30 = Multiply(A=@n28, B=ScalarParameter() {defaultValue=1.5, parameterName=LoudKiwi.OnReloadBoost}) {constB=1.3}
  n31 = NamedRerouteUsage()
  n32 = Power(Base=NamedRerouteUsage()) {constExponent=0.3}
  n33 = LinearInterpolate(
   A=LinearInterpolate(A=Multiply(A=@n16, B=@n32), B=@n8.RGB, Alpha=NamedRerouteUsage())
   Alpha=SmoothStep(Value=NamedRerouteUsage()) {constMax=0.05}
  ) {constB=0, desc=Hide glow dureing reload anim}
  n34 = ScalarParameter() {defaultValue=0.8, group=Loud Kiwi, parameterName=LoudKiwi.Glow.Bend}
  n35 = Subtract(A=NamedRerouteUsage(), B=NamedRerouteUsage())
  n36 = VectorParameter() {group=Loud Kiwi, parameterName=LoudKiwi.Base.Color, sortPriority=1, defaultValue=<rgba 0.1413 0.0075 0.2747 1>}
  n37 = Divide(A=ScalarParameter() {defaultValue=1, group=Loud Kiwi, parameterName=LoudKiwi.PanningNoiseMaskRange, sortPriority=30})
  n38 = ScalarParameter() {group=Loud Kiwi, parameterName=LoudKiwi.PanningNoiseBlendBias, sortPriority=30}
  n39 = Add(A=Subtract(B=@n37) {constA=0.5}, B=@n38)
  n40 = Add(A=Add(A=@n37) {constB=0.5}, B=@n38)
  n41 = Power(Base=ComponentMask(None=NamedRerouteUsage()) {g=true})
  n42 = SetMaterialAttributes!(
   MaterialAttributes
   BaseColor=LinearInterpolate(A=NamedRerouteUsage(), B=@n36.RGB, Alpha=SmoothStep(Min=@n39, Max=@n40, Value=@n41) {constMax=0.456, constMin=0.2}) {constAlpha=0.8}
   Metallic=ScalarParameter() {defaultValue=0.15, group=Loud Kiwi, parameterName=LoudKiwi.Metalness, sortPriority=35}
   EmissiveColor=LinearInterpolate(A=LinearInterpolate(A=@n29, B=@n30, Alpha=@n31), B=LinearInterpolate(B=@n33, Alpha=@n34), Alpha=Saturate(None=@n35))
   Roughness=NamedRerouteUsage()
   Normal=MaterialFunctionCall(
    BaseNormal=GetMaterialAttributes(None=@n9) {attributeGetTypes [#1]}.Normal
    AdditionalNormal=NamedRerouteUsage()
   ) {materialFunction=&BlendAngleCorrectedNormals}.Result
   CustomData0=@n10
   CustomData1=@n10
  ) {attributeSetTypes [
    #2
    #3
    #4
    #5
    #1
    #6
    #7
   ]}
  NamedRerouteDeclaration(None=@n42) {name=Material, nodeColor=<rgba 0.0547 1 0 1>}
  n43 = Multiply(A=NamedRerouteUsage(), B=ScalarParameter() {defaultValue=1, group=Loud Kiwi, parameterName=LoudKiwi.NormalTiling})
  n44 = ScalarParameter() {defaultValue=1, group=Loud Kiwi, parameterName=LoudKiwi.NormalIntensity}
  NamedRerouteDeclaration(None=MaterialFunctionCall(
    Normal=TextureSample(UVs=@n43) {samplerType=SAMPLERTYPE_Normal, texture=/PepperCatShared/Assets/TechArt/T_PepperCatNoise_N}.RGB
    Flatness=OneMinus(None=@n44)
   ) {materialFunction=&FlattenNormal}.Result) {name=Normal, nodeColor=<rgba 0.9922 1 0 1>}
  n45 = TextureSample(UVs=NamedRerouteUsage()) {samplerType=SAMPLERTYPE_Masks, texture=/Game/Global/Textures/Default/Noise/T_CubeNoise}
  NamedRerouteDeclaration(None=@n45.R) {name=MaskUnprocessed, nodeColor=<rgba 0 1 0.6953 1>}
  n46 = LocalPosition() {localOrigin=InstancePreSkinning}
  n47 = VectorParameter() {group=Loud Kiwi, parameterName=LoudKiwi.ShellGlow.BoxMaskBounds, defaultValue=<rgba 6 21.4749 4.7463 1>}
  n48 = MaterialFunctionCall!(
   A=@n46.XYZ
   B=VectorParameter() {group=Loud Kiwi, parameterName=LoudKiwi.ShellGlow.BoxMaskPos, defaultValue=<rgba 0 10.6605 13.8856 1>}.RGB
   Bounds=@n47.RGB
   Edge Falloff=ScalarParameter() {defaultValue=1, group=Loud Kiwi, parameterName=LoudKiwi.ShellGlow.Hardness}
  ) {materialFunction=/Engine/Functions/Engine_MaterialFunctions01/Gradient/BoxMask-3D}
  n49 = Normalize(VectorInput=Subtract(A=VectorParameter() {group=Loud Kiwi, parameterName=LoudKiwi.ShellGlow.Pos, defaultValue=<rgba 0 7.1085 14.7443 1>}.RGB, B=@n46.XYZ))
  n50 = Saturate(None=DotProduct(A=Transform(None=@n49) {transformSourceType=TRANSFORMSOURCE_Local}, B=VertexNormalWS()))
  n51 = NamedRerouteUsage()
  n52 = Saturate(None=Subtract(A=Multiply(A=@n48.Result, B=@n50), B=OneMinus(None=@n51)))
  NamedRerouteDeclaration(None=ShaderStageSwitch(PixelShader=VertexInterpolator(VS=@n52).PS, VertexShader=@n52)) {name=ShellGlowMask, nodeColor=<rgba 0 0.625 1 1>}
  NamedRerouteDeclaration(None=Saturate(None=ScalarParameter() {group=Loud Kiwi, parameterName=LoudKiwi.BP_PowerLevel, sliderMax=1, sortPriority=10})) {name=PowerLevelProcessed, nodeColor=<rgba 0.2656 0 1 1>}
  NamedRerouteDeclaration(None=@n3) {name=MainMask, nodeColor=<rgba 0.9453 1 0 1>}
  NamedRerouteDeclaration(None=@n45.RGB) {name=MaskRGB, nodeColor=<rgba 0 1 0.3672 1>}
  n53 = Subtract(A=Time(), B=ScalarParameter() {defaultValue=-10, group=Loud Kiwi, parameterName=LoudKiwi.BP_TimeReload, sortPriority=10})
  n54 = ScalarParameter() {defaultValue=0.25, group=Loud Kiwi, parameterName=LoudKiwi.ReloadPulseDuration, sortPriority=11}
  n55 = Power(
   Base=Saturate(None=Divide(A=@n53, B=@n54))
   Exp=ScalarParameter() {defaultValue=1, group=Loud Kiwi, parameterName=LoudKiwi.FirePulseCurve, sortPriority=11}
  )
  NamedRerouteDeclaration(None=LinearInterpolate(A=OneMinus(None=@n55), Alpha=ScalarParameter() {group=Loud Kiwi, parameterName=LoudKiwi.DebugReloadAlpha, sliderMax=1, sortPriority=11})) {name=BP_ReloadAlpha, nodeColor=<rgba 0 0.3438 1 1>}
  NamedRerouteDeclaration(None=FunctionInput() {inputName=TechArtMask, sortPriority=1}) {name=TAMask, nodeColor=<rgba 0.7109 0 1 1>}
  n56 = TextureCoordinate()
  n57 = NamedRerouteUsage()
  n58 = Multiply(A=ComponentMask(None=Subtract(A=@n57) {constB=0.5}) {g=true, r=true}) {constB=0.01}
  n59 = Power(Base=TextureSample(UVs=Add(A=@n56, B=@n58)) {texture=/LoudKiwi/Assets/Textures/T_LoudKiwi_Cube_SMG_FX_2}.R) {constExponent=0.3}
  n60 = ScalarParameter() {defaultValue=1, group=Loud Kiwi, parameterName=LoudKiwi.AccentIntensity, sortPriority=6}
  NamedRerouteDeclaration(None=Saturate(None=Multiply(A=@n59, B=@n60))) {name=AccentMask, nodeColor=<rgba 0 1 0.3906 1>}
  n61 = Subtract(A=Time(), B=ScalarParameter() {defaultValue=-10, group=Loud Kiwi, parameterName=LoudKiwi.BP_TimeOnFire, sortPriority=10})
  n62 = ScalarParameter() {defaultValue=0.25, group=Loud Kiwi, parameterName=LoudKiwi.FirePulseDuration, sortPriority=11}
  n63 = Power(
   Base=Saturate(None=Divide(A=@n61, B=@n62))
   Exp=ScalarParameter() {defaultValue=1, group=Loud Kiwi, parameterName=LoudKiwi.FirePulseCurve, sortPriority=11}
  )
  NamedRerouteDeclaration(None=LinearInterpolate(A=OneMinus(None=@n63), Alpha=ScalarParameter() {group=Loud Kiwi, parameterName=LoudKiwi.DebugFireAlpha, sliderMax=1, sortPriority=11})) {name=BP_FireAlpha, nodeColor=<rgba 0 0.3437 1 1>}
  NamedRerouteDeclaration(None=@n61) {name=TimeSinceFired, nodeColor=<rgba 0 0.4609 1 1>}
  NamedRerouteDeclaration(None=LinearInterpolate(
    A=VectorParameter() {group=Loud Kiwi, parameterName=LoudKiwi.PanningNoiseColorA, defaultValue=<rgba 0.4072 0.0232 0.5333 1>}.RGB
    B=VectorParameter() {group=Loud Kiwi, parameterName=LoudKiwi.PanningNoiseColorB, defaultValue=<rgba 0.2423 0.0137 0.3185 1>}.RGB
    Alpha=NamedRerouteUsage()
   )) {name=PanningNoiseColor, nodeColor=<rgba 1 0.5859 0 1>}
  n64 = NamedRerouteUsage()
  n65 = Constant() {r=0.02}
  n66 = Constant()
  n67 = Constant() {r=0.605}
  n68 = BumpOffset(Coordinate=TextureCoordinate(), Height=ScalarParameter() {group=Loud Kiwi, parameterName=LoudKiwi.Glow.BumpOffset})
  n69 = MaterialFunctionCall!(Width=@n65, Sharpness=@n66, Offset=Constant() {r=0.665}, Direction Switch, Compare, Input Coordinates=@n68) {materialFunction=/Engine/Functions/Engine_MaterialFunctions03/Procedurals/GeneratedBand}
  n70 = MaterialFunctionCall(Width=@n65, Sharpness=@n66, Offset=Constant() {r=0.55}, Input Coordinates=@n68) {materialFunction=&GeneratedBand}
  n71 = Power(
   Base=Max(
    A=MaterialFunctionCall(Width=@n65, Sharpness=@n66, Offset=@n67, Input Coordinates=@n68) {materialFunction=&GeneratedBand}.Result
    B=Max(A=@n69.Result, B=@n70.Result)
   )
   Exp=ScalarParameter() {defaultValue=1, group=Loud Kiwi, parameterName=LoudKiwi.Glow.Contrast}
  )
  n72 = ScalarParameter() {defaultValue=1, group=Loud Kiwi, parameterName=LoudKiwi.Glow.Fresnel.Brighness}
  n73 = ScalarParameter() {defaultValue=1, group=Loud Kiwi, parameterName=LoudKiwi.Glow.Fresnel.Exponent}
  NamedRerouteDeclaration(None=Multiply(
    A=Multiply(A=@n64, B=@n71)
    B=MaterialFunctionCall(Brightness=@n72, Exponent=@n73) {materialFunction=&MF_Fresnel_FX}.Fresnel
   )) {name=InternalGradients, nodeColor=<rgba 0 0.6719 1 1>}
  n74 = ScalarParameter() {group=Loud Kiwi, parameterName=LoudKiwi.RoughnessMin, sortPriority=35}
  n75 = ScalarParameter() {defaultValue=1, group=Loud Kiwi, parameterName=LoudKiwi.RoughnessMax, sortPriority=35}
  n76 = NamedRerouteUsage()
  NamedRerouteDeclaration(None=Saturate(None=LinearInterpolate(A=@n74, B=@n75, Alpha=@n76))) {name=RoughnessOut, nodeColor=<rgba 1 0.6563 0 1>}
  NamedRerouteDeclaration(None=@n1) {name=MA_Source, nodeColor=<rgba 1 0 0.0469 1>}
  n77 = Multiply(
   A=BumpOffset(Coordinate=NamedRerouteUsage(), Height=ScalarParameter() {defaultValue=-1, group=Loud Kiwi, parameterName=LoudKiwi.PanningNoiseBumpOffset})
   B=ScalarParameter() {defaultValue=1, group=Loud Kiwi, parameterName=LoudKiwi.NoiseTiling, sortPriority=30}
  )
  n78 = Panner(Coordinate=Multiply(A=@n77) {constB=8}, Time=Multiply(A=Time()) {constB=0.4}) {speedX=-1, speedY=0.5}
  n79 = Time()
  n80 = TextureSample(UVs=Panner(Coordinate=Multiply(A=@n77) {constB=3}, Time=Multiply(A=@n79) {constB=0.075}) {speedX=-1, speedY=0.5}) {samplerType=SAMPLERTYPE_Grayscale, texture=&LowResBlurredNoise}
  n81 = Multiply(A=@n77) {constB=2}
  n82 = Multiply(A=Time()) {constB=0.1}
  n83 = Multiply(
   A=@n80.R
   B=TextureSample(UVs=Panner(Coordinate=@n81, Time=@n82) {speedX=1, speedY=-0.5}) {samplerType=SAMPLERTYPE_Grayscale, texture=&LowResBlurredNoise}.R
  )
  n84 = Multiply(B=@n83) {constA=1.5}
  n85 = Multiply(A=NamedRerouteUsage(), B=ScalarParameter() {defaultValue=0.1, group=Loud Kiwi, parameterName=LoudKiwi.NoiseWarp}) {constB=0.2}
  n86 = TextureSample(UVs=Add(A=Add(A=@n78, B=@n84), B=ComponentMask(None=@n85) {g=true, r=true})) {samplerType=SAMPLERTYPE_Grayscale, texture=&LowResBlurredNoise}
  n87 = Multiply(A=Multiply(A=Power(Base=@n86.R) {constExponent=3}) {constB=12}, B=@n80.R)
  NamedRerouteDeclaration(None=Saturate(None=Add(A=@n87, B=@n83))) {name=PanningNoiseMask, nodeColor=<rgba 1 0 0.9844 1>}
  NamedRerouteDeclaration(None=Multiply(A=TextureCoordinate(), B=ScalarParameter() {defaultValue=1, group=Loud Kiwi, parameterName=LoudKiwi.TilingAll})) {name=UVs, nodeColor=<rgba 1 0 0.1406 1>}
 ]}
