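# Skeletal mesh master material for the LoudKiwi Cube SMG, layering MF_LoudKiwi over the shared Fortnite weapon master function with a TechArtMask.
# A selectable mask channel blends in a clear coat variant with adjustable undercoat metallic and roughness, where one parameter swaps the roughness map between base and coat.
# PlatformSwitch_MA limits that blend to high-end PC and current-gen consoles; other tiers receive the uncoated result.
# First-person interpolation is fixed fully on.
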Material /LoudKiwi/Assets/Materials/M_LoudKiwi_Cube_SMG_P
graph {
 properties {
  bAutomaticallySetUsageInEditor=false
  bUseMaterialAttributes=true
  bUsedWithMeshDeformer=true
  bUsedWithMorphTargets=true
  bUsedWithSkeletalMesh=true
  pixelDepthOffsetMode=PDOM_Legacy
  shadingModel=MSM_FromMaterialExpression
  usedShadingModels=Clear Coat
 }
 nodes [
  FirstPersonOutput!(FirstPersonInterpolationAlpha=Constant() {r=1})
  n1 = Constant()
  n2 = TextureSampleParameter2D() {
   group=05_Tech Art Masks
   parameterName=TechArtMask
   samplerSource=SSM_Wrap_WorldGroupSettings
   samplerType=SAMPLERTYPE_Masks
   sortPriority=0
   texture=&T_BlackMask
  }
  n3 = MaterialFunctionCall!(
   Material=SetMaterialAttributes!(
    MaterialAttributes=MaterialFunctionCall!(In) {materialFunction=/Game/Packages/Fortress_SharedMaterials/Base_Material_Functions/BaseCustomizationFunctions/MF_FN_Weapon_MASTER}.Material Result
    Clear Coat=@n1
    Clear Coat Roughness=@n1
   ) {attributeSetTypes [#1, #2]}
   TechArtMask=@n2.RGB
  ) {materialFunction=&/MF_LoudKiwi}
  n4 = GetMaterialAttributes(None=@n3.Material) {attributeGetTypes [#3, #4]}
  n5 = ScalarParameter() {defaultValue=0.1, group=06_Clearcoat, parameterName=UnderCoatRoughness}
  n6 = ScalarParameter() {group=06_Clearcoat, parameterName=RoughnessMapAffectClearCoatRoughness}
  n7 = SetMaterialAttributes!(
   MaterialAttributes=@n4.MaterialAttributes
   Clear Coat=Constant() {r=1}
   Clear Coat Roughness=LinearInterpolate(A=@n5, B=@n4.Roughness, Alpha=@n6)
   Roughness=LinearInterpolate(A=@n4.Roughness, B=@n5, Alpha=@n6)
   Metallic=Multiply(A=@n4.Metallic, B=ScalarParameter() {defaultValue=1, group=06_Clearcoat, parameterName=UnderCoatMetallicMultiplier})
   Shading Model=ShadingModel() {shadingModel=MSM_ClearCoat}
  ) {attributeSetTypes [
    #1
    #2
    #3
    #4
    #5
   ]}
  MaterialFunctionCall!(
   "Xb1,PS4"=@n3.Material
   "PC-High&&Epic,XSX,PS5"=BlendMaterialAttributes(A=@n4.MaterialAttributes, B=@n7, Alpha=ChannelMaskParameter(None=@n2.RGB) {group=05_Tech Art Masks, parameterName=ClearCoatMaskChannel}.R)
   "PC-Perf,Android"=@n3.Material
   Switch=@n3.Material
   "PC-Low(NotPerf)&&Medium"=@n3.Material
  ) {materialFunction=/Game/Effects/Niagara/MasterMaterials/Functions/Common/PlatformSwitch_MA}
 ]
}
