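# Niagara character-hit effect for the Cube Splitter, covering a skeletal mesh's body in cubes, runes, lightning and purple decals.
# AttachCubes and Decals follow filtered bones via skinned mesh sampling, with curl-noise vector fields displacing the cubes.
# BurstRunes and Lightning spawn from AttachCubes particles through attribute readers.
# An integer SpawnDensity scales cube spawn rate, glow and opacity, decal brightness, and gates lightning; Light and LightRays have no renderers.
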NiagaraSystem /LoudKiwi/Assets/NS_CubeSplitter_CharHit
dependencies {
 usedDataInterfaces [
  &NiagaraDataInterfaceArrayFloat
  /Script/Niagara.NiagaraDataInterfaceVectorField
  &NiagaraDataInterfaceCurve
  &NiagaraDataInterfaceColorCurve
  &NiagaraDataInterfaceParticleRead
  &NiagaraDataInterfaceSpriteRendererInfo
  &NiagaraDataInterfaceCamera
  &NiagaraDataInterfaceArrayInt32
 ]
 usedDynamicInputs [
  &SimulationPosition
  &Vector2DFromFloat
  &MakeVectorFromLinearColorRGB
  &TransformPosition
 ]
 usedModules [
  &self:SystemSpawnScript.SystemScriptSource.SystemScriptGraph.NiagaraNodeAssignment_0.SetVariables_#1
  &self:SystemSpawnScript.SystemScriptSource.SystemScriptGraph.NiagaraNodeAssignment_1.SetVariables_#2
  &SystemState
  &self:SystemSpawnScript.SystemScriptSource.SystemScriptGraph.NiagaraNodeAssignment_2.SetVariables_#3
  &self:SystemSpawnScript.SystemScriptSource.SystemScriptGraph.NiagaraNodeAssignment_3.SetVariables_#4
 ]
}
emitters {
 AttachCubes {
  emitterUpdateScript [
   &EmitterState()
   &SpawnRate(SpawnRate=&Multiply_FloatByInt(Float=10, Integer=$System.SpawnDensity), Use Spawn Probability=false)
   /SetVariables(Emitter.Glow:NiagaraFloat=&Lerp_Float(
     A=&SelectFloatFromArray(Direct Array Index=$System.SpawnDensity, Float Selection Array=DataInterface() {propertyValues {FloatData [5,10,20,30]}})
     B=&Waveform(
      "Waveform [1]"=::NewEnumerator10
      "[1] Phase"=&Add_Float(A=$Emitter.Age, B=&RandomRangeFloat(Minimum=0, Maximum=1))
      "[1] Frequency"=1
      "[1] Amplitude Min/Max"=<10 150>
     )
     Alpha=&MakeCustomFloatFromBool(
      Bool=&SetBoolByIntComparison(A=$System.SpawnDensity, B=4, Comparison Type=::NewEnumerator4)
      True Float=1
      False Float=0
     )
    )) {isSetParameters=true}
  ]
  particleSpawnScript [
   &PartitionParticles(Number Of Partitions=6)
   &InitializeParticle(
    Color Mode=::NewEnumerator0
    Mesh Scale=&Multiply_VectorByFloat(Vector=<2 2 2>, Float=&RandomRangeFloat(Minimum=0.5, Maximum=2))
    Mesh Scale Mode=::NewEnumerator4
    Sprite Size Mode=::NewEnumerator3
    Uniform Sprite Size=30
    UsePositionOffset=false
    Write Material Random=true
   )
   /SetVariables() {isSetParameters=true}
   &self:AddNoise_01(
    Noise Frequency=12
    Noise Magnitude:NiagaraFloat=&Multiply_Float(A=1, B=8)
    Vector Field=DataInterface() {propertyValues {Field=/NiagaraStaging/VectorFields/TilingCurl16, bTileX=true, bTileY=true, bTileZ=true}}
   ) {module=AddNoise_01}
   &self:AddNoise_01(
    Noise Frequency=25
    Noise Magnitude:NiagaraFloat=&Multiply_Float(A=1, B=12)
    Vector Field=DataInterface() {propertyValues {Field=&TilingCurl16, bTileX=true, bTileY=true, bTileZ=true}}
   ) {module=AddNoise_01001}
   /SetVariables(Particles.Random=&RandomRangeFloat(Minimum=0, Maximum=1)) {isSetParameters=true}
  ]
  particleUpdateScript [
   &SampleSkeletalMesh(
    AdditionalYawPitchRoll=false
    Bone / Socket Index:NiagaraInt32=$Particles.PartitionParticles.Partition
    Bone Sampling Mode:ENiagaraBoneSamplingMode=::NewEnumerator4
    Skeletal Mesh=DataInterface() {propertyValues {
      DefaultMesh=null
      ExcludeBoneName=root
      OptionalComponentClass=null
      PreviewMesh=/Game/Characters/Player/Female/Medium/Bodies/F_MED_Ramirez_Fallback/Meshes/F_MED_Ramirez_Fallback
      SkinningMode=SkinOnTheFly
      SoftSourceActor=null
      SourceMode=Default
      UvSetIndex=0
      WholeMeshLOD=-1
      bExcludeBone=false
      bReadDeformedGeometry=true
      bRequireCurrentFrameData=true
      ComponentTags []
      FilteredBones [spine_03,spine_04,upperarm_l,upperarm_r,spine_02,pelvis]
      FilteredSockets []
      MeshUserParameter {parameter {name=None, typeDefHandle {registeredTypeIndex=1}, varData []}}
      SamplingRegions []
     }}
   )
   &ShapeLocation(
    Additional Quaternion Rotation=false
    Cylinder Height=5
    Cylinder Radius=5
    MaskSpawnGroup=false
    Override Seed=false
    Rotation Mode=::NewEnumerator3
    Rotation Quaternion=$Particles.SampleSkeletalMesh.SampledRotation
    Shape Origin=$Particles.SampleSkeletalMesh.SampledPosition
    Shape Primitive=::NewEnumerator1
    Surface Only=false
    UseOwnerScale=true
   )
   /SetVariables(Particles.RadialSignal:NiagaraFloat=&FloatFromCurve(
     FloatCurve=DataInterface() {propertyValues {Curve {
        defaultValue=0
        keys [
         {arriveTangent=9.3366, interpMode=RCIM_Cubic, leaveTangent=9.3366, tangentMode=RCTM_User, time=0, value=0}
         {arriveTangent=-1.0366, interpMode=RCIM_Cubic, leaveTangent=-1.0366, tangentMode=RCTM_User, time=0.25, value=0.9973}
        ]
       }}}
     CurveIndex=&ClampFloat(Float=&Add_Float(
       A=&Power(
        a=&OneMinusFloat(Float=$Particles.Falloff)
        b=&RandomRangeFloat(Minimum=0.75, Maximum=1)
       )
       B=&Subtract_Float(
        A=&Multiply_Float(A=$Engine.System.Age, B=&RandomRangeFloat(Minimum=0.5, Maximum=4))
        B=1
       )
      ))
    )) {isSetParameters=true}
   /SetVariables(Particles.Position=&AddVectorToPosition(
     Position=$Particles.Position
     Vector=&Multiply_VectorByFloat(Vector=$Particles.SummedNoise, Float=$Particles.RadialSignal)
    )) {isSetParameters=true}
   &ParticleState()
   &CameraOffset(Camera Offset Amount=30)
   /SetVariables(Particles.Scale=&Multiply_VectorByFloat(
     Vector=$Particles.Initial.Scale
     Float=&Multiply_Float(
      A=&OneMinusFloat(Float=$Particles.Falloff)
      B=&Multiply_Float(A=$Particles.RadialSignal, B=1.5)
     )
    )) {isSetParameters=true}
   &ScaleMeshSize(Scale Factor=&VectorFromFloat(Value=&FloatFromCurve(
      FloatCurve=DataInterface() {propertyValues {Curve {keys [<at 0 0>,<at 0.0601 1.0039>,<at 0.2582 1.0147>,<at 0.2975 0.7951>,<at 0.4043 0.8029>,<at 0.4815 0.3245>,<at 0.5728 0.3088>,<at 0.6276 0.0971>,<at 0.7133 0.0735>,<at 1 0>]}}}
      Scale Curve=1.5
     )))
   &FN_Sprite_Attributes(
    Camera Offset=50
    Color=&MultiplyLinearColor(
     Linear Color=&ColorFromCurve(ColorCurve=DataInterface() {propertyValues {
        AlphaCurve {keys [<at 0 1 RCIM_Cubic>,<at 1 0 RCIM_Cubic>]}
        BlueCurve {keys [<at 0 1 RCIM_Cubic>]}
        GreenCurve {keys [<at 0 0 RCIM_Cubic>]}
        RedCurve {keys [<at 0 0.4447 RCIM_Cubic>]}
       }})
     Scale Factor=&MakeVector4(
      X=1
      Y=1
      Z=1
      W=&SelectFloatFromArray(Direct Array Index=$System.SpawnDensity, Float Selection Array=DataInterface() {propertyValues {FloatData [0,0.5,0.65,0.7,0.8]}})
     )
    )
    Create Camera Offset Attribute=true
    Create Color Attribute=true
    Module Output:ENiagara_AttributeSamplingApplyOutput=::NewEnumerator1
   )
  ]
  rendererClasses [&NiagaraMeshRendererProperties, &NiagaraSpriteRendererProperties]
  renderers [
   {rendererClass=&NiagaraMeshRendererProperties, rendererIndex=0}
   {rendererClass=&NiagaraSpriteRendererProperties, rendererIndex=1}
  ]
 }
 BurstRunes {
  emitterUpdateScript [
   &EmitterState()
   &SpawnParticlesFromOtherEmitter(
    Attribute Reader=DataInterface() {propertyValues {EmitterBinding {bindingMode=Other, emitterName=AttachCubes}}}
    Calculate Spawn Rate Per Particle=false
    Minimum Age Spawn Threshold=1
    Use Maximum Age=false
    Use Minimum Age=false
   )
  ]
  particleSpawnScript [
   &InitializeParticle(
    Color=&MakeLinearColor(
     R:NiagaraFloat=&UniformAOrBFloat(A=0, Evaluation Type=::NewEnumerator0, Recalculate Random Each Loop=true, Override Seed=false)
     G:NiagaraFloat=&UniformRangedFloat(Minimum=8, Maximum=10)
     B:NiagaraFloat=&UniformRangedFloat(Minimum=3, Maximum=4)
     A:NiagaraFloat=&UniformRangedFloat(Minimum=0.5, Maximum=2)
    )
    Sprite Size Mode=::NewEnumerator3
    UsePositionOffset=false
    Write Material Random=true
   )
   &SampleParticlesFromOtherEmitter(
    Applied Position Offset:Vector3f=<0 0 5>
    Mesh Scale Sampling:ENiagara_DirectReadSamplingMode=::NewEnumerator0
    Reject Particle If Sampled Index Is Invalid=false
   )
   /SetVariables(Particles.SpriteInitRotationRate:NiagaraFloat=&Multiply_Float(
     A=&UniformAOrBFloat(A=-1, Evaluation Type=::NewEnumerator0, Recalculate Random Each Loop=true, Override Seed=false)
     B=&UniformRangedFloat(Minimum=5, Maximum=10)
    )) {isSetParameters=true}
   /SetVariables(Particles.SpriteSize=&Multiply_Vector2D_ByFloat(
     Float=&ClampFloat(Float=&MakeFloatFromVector(Vector=$Particles.Scale, Channel=::NewEnumerator0))
     Vector2D=$Particles.SpriteSize
    )) {isSetParameters=true, module=SetVariables_#5}
   &SubUVAnimation(
    Evaluation Type=::NewEnumerator0
    "Sprite Renderer "=DataInterface() {propertyValues {SpriteRenderer=&self:BurstRunes_3.NiagaraSpriteRendererProperties_2}}
    SubUV Animation Mode:ENiagara_SubUVLookupModeV2=::NewEnumerator2
    UseEndFrame=false
    UseStartFrame=false
   )
   &ShapeLocation(MaskSpawnGroup=false, Override Seed=false, Sphere Radius=25, UseOwnerScale=true)
  ]
  particleUpdateScript [
   &ParticleState()
   &SpriteRotationRate(Rotation Rate=&Multiply_Float(A=$Particles.SpriteInitRotationRate, B=&VectorLength(Vector=$Particles.Velocity)))
   &ScaleColor(
    Scale Alpha=&FloatFromCurve(FloatCurve=DataInterface() {propertyValues {Curve {keys [{interpMode=RCIM_Cubic, tangentMode=RCTM_User, time=0.75, value=1}, <at 1 0.5>]}}})
    Scale RGB=&MakeVector(
     X=1
     Y=1
     Z=&FloatFromCurve(FloatCurve=DataInterface() {propertyValues {Curve {keys [
          {arriveTangent=-32.9921, interpMode=RCIM_Cubic, leaveTangent=-32.9921, tangentMode=RCTM_User, time=0.2, value=10}
          {arriveTangent=-0.0507, interpMode=RCIM_Cubic, leaveTangent=-0.0508, tangentMode=RCTM_User, time=0.3, value=1}
          {arriveTangent=8.5796, interpMode=RCIM_Cubic, leaveTangent=8.5796, tangentMode=RCTM_User, time=1, value=20}
         ]}}})
    )
    ScaleA=true
    ScaleRGB=true
   )
   &CameraOffset(Camera Offset Amount=50)
   &ScaleSpriteSize(
    Scale Sprite Size Mode=::NewEnumerator0
    Uniform Scale Factor=&Lerp_Float(
     A=0.5
     B=1
     Alpha=&NormalizeDistanceRange(
      Start Position:NiagaraPosition=$Particles.Position
      End Position:NiagaraPosition=&ReturnCameraVector(Camera Vectors=::NewEnumerator1, Coordinate Space=::Simulation)
      Distance:NiagaraFloat=200
      Invert Normalized Range:NiagaraBool=false
     )
    )
   )
  ]
  rendererClasses [&NiagaraSpriteRendererProperties]
  renderers [{rendererClass=&NiagaraSpriteRendererProperties}]
 }
 Decals {
  emitterUpdateScript [
   &EmitterState()
   &SpawnRate(SpawnRate=10, Use Spawn Probability=false)
  ]
  particleSpawnScript [
   /SetVariables(Particles.Index:NiagaraInt32=&SelectIntFromArray(
     Direct Array Index=&ReturnParticleIDAttributes()
     Int Selection Array=DataInterface() {propertyValues {IntData [0,1,2,8,9,17,18]}}
    )) {isSetParameters=true}
   &InitializeParticle(
    Color=<rgba 0.4583 0 1 1>
    Lifetime=2
    Sprite Size Mode=::NewEnumerator3
    Uniform Sprite Size=100
    UsePositionOffset=false
    Write Material Random=true
   )
   &CameraOffset(Camera Offset Amount=32)
  ]
  particleUpdateScript [
   &ParticleState()
   &SampleSkeletalMesh(
    AdditionalYawPitchRoll=false
    Bone / Socket Index:NiagaraInt32=&ModuloInt(A=&ReturnParticleIDAttributes(), B=11)
    Bone Sampling Mode:ENiagaraBoneSamplingMode=::NewEnumerator4
    Skeletal Mesh=DataInterface() {propertyValues {
      DefaultMesh=null
      ExcludeBoneName=root
      OptionalComponentClass=null
      PreviewMesh=&F_MED_Ramirez_Fallback
      SkinningMode=SkinOnTheFly
      SoftSourceActor=null
      SourceMode=Default
      UvSetIndex=0
      WholeMeshLOD=-1
      bExcludeBone=false
      bReadDeformedGeometry=true
      bRequireCurrentFrameData=true
      ComponentTags []
      FilteredBones [spine_01,spine_03,spine_05,upperarm_l,lowerarm_l,upperarm_r,lowerarm_r,thigh_l,calf_l,thigh_r,calf_r]
      FilteredSockets []
      MeshUserParameter {parameter {name=None, typeDefHandle {registeredTypeIndex=1}, varData []}}
      SamplingRegions []
     }}
   )
   &ShapeLocation(
    Cylinder Height=15
    Cylinder Radius=5
    MaskSpawnGroup=false
    Override Seed=false
    Shape Origin=$Particles.SampleSkeletalMesh.SampledPosition
    Shape Primitive=::NewEnumerator1
    Surface Only=false
    UseOwnerScale=true
   )
   &DynamicMaterialParameters(
    Index 0 Param 1=4
    Index 0 Param 2=1
    Index 0 Param 3=&Multiply_Float(A=$Particles.Age, B=0.1)
    Index 0 Param 4=&RandomRangeFloat(Minimum=0.75, Maximum=0.9)
    Index 1 Param 1=250
    Index 1 Param 2=1
    Module.Param0WriteEnabled=true
    Module.Param0WriteEnabled1=true
    Module.Param1WriteEnabled=true
    Module.Param1WriteEnabled1=true
    Module.Param2WriteEnabled=true
    Module.Param2WriteEnabled1=false
    Module.Param3WriteEnabled=true
    Module.Param3WriteEnabled1=false
    Write Parameter Index 1=true
   )
   &ScaleColor(
    Scale RGB=&VectorFromFloat(Value=&FloatFromCurve(FloatCurve=DataInterface() {propertyValues {Curve {keys [
          <at 0 0 RCIM_Cubic>
          {arriveTangent=-0.4449, interpMode=RCIM_Cubic, leaveTangent=-0.4449, tangentMode=RCTM_User, time=0.5, value=2.9294}
          <at 1 0>
         ]}}}))
    ScaleA=true
    ScaleRGB=true
   )
   &ScaleColor(
    Curve Index=&SelectFloatFromArray(Direct Array Index=$System.SpawnDensity, Float Selection Array=DataInterface() {propertyValues {FloatData [0,0.25,0.35,0.5,0.75]}})
    Linear Color Curve=DataInterface() {propertyValues {
      AlphaCurve {keys [<at 0 1>]}
      BlueCurve {keys [<at 0 0.25>,<at 0.2529 0.35>,<at 0.5372 0.5>,<at 0.7754 0.75>,<at 1 1.5>]}
      GreenCurve {keys [<at 0 0.25>,<at 0.2529 0.35>,<at 0.5372 0.5>,<at 0.7754 0.75>,<at 1 1.5>]}
      RedCurve {keys [<at 0 0.25>,<at 0.2529 0.35>,<at 0.5372 0.5>,<at 0.7754 0.75>,<at 1 1.5>]}
     }}
    Scale Mode=2
   ) {module=ScaleColor001}
   &FN_Sprite_Attributes(
    Camera Offset=-50
    Color=&ColorFromCurve(ColorCurve=DataInterface() {propertyValues {
       AlphaCurve {keys [<at 0.0022 0>,<at 0.4508 0.5 RCIM_Cubic>,<at 1 0 RCIM_Cubic>]}
       BlueCurve {keys [<at 0.4437 1 RCIM_Cubic>]}
       GreenCurve {keys [<at 0.4437 0.0608 RCIM_Cubic>]}
       RedCurve {keys [<at 0.4437 0.7969 RCIM_Cubic>]}
      }})
    Create Camera Offset Attribute=true
    Create Color Attribute=true
    Create Dynamic Material Parameter Attributes=true
    Create SpriteSize Attribute=true
    Module Output:ENiagara_AttributeSamplingApplyOutput=::NewEnumerator1
    Module.Index0Float0Write=true
    Module.Index0Float1Write=true
    Module.Index0Float2Write=true
    Module.Index0Float3Write=true
    SpriteSize=<75 75>
    Write Parameter Index 1:ENiagara_DynamicMaterialParametersInputMode=::NewEnumerator1
   )
  ]
  rendererClasses [&NiagaraSpriteRendererProperties]
  renderers [
   {rendererClass=&NiagaraSpriteRendererProperties}
   {rendererClass=&NiagaraSpriteRendererProperties, rendererIndex=1}
  ]
 }
 Light {rendererClasses [], renderers []}
 LightRays {rendererClasses [], renderers []}
 Lightning {
  emitterUpdateScript [
   &EmitterState()
   &SpawnParticlesFromOtherEmitter(
    Attribute Reader=DataInterface() {propertyValues {EmitterBinding {bindingMode=Other, emitterName=AttachCubes}}}
    Calculate Spawn Rate Per Particle=false
    Minimum Age Spawn Threshold=1
    Spawn Rate=&FloatFromCurve(
     FloatCurve=DataInterface() {propertyValues {Curve {
        defaultValue=1
        keys [
         {arriveTangent=9.0725, time=0, value=3}
         {arriveTangent=-2.5249, time=0.2267, value=1.0588}
         <at 0.6798 0>
         {arriveTangent=-0.0405, leaveTangent=-0.0405, time=2, value=0}
        ]
       }}}
     CurveIndex=$Emitter.NormalizedLoopAge
     Scale Curve=&Multiply_Float(
      A=10
      B=&MakeCustomFloatFromBool(
       Bool=&SetBoolByIntComparison(A=$System.SpawnDensity, B=4, Comparison Type=::NewEnumerator4)
       True Float=1
       False Float=0
      )
     )
    )
    Use Maximum Age=false
    Use Minimum Age=false
   )
  ]
  particleSpawnScript [
   &InitializeParticle(
    Color=&MultiplyLinearColor(
     Linear Color=&RandomRangeLinearColor(Minimum=<rgba 0.5306 0 1 0.3>, Maximum=<rgba 0.9352 0 1.8 1>)
     Scale Factor=&MakeVector4_FromFloat(Float=0.8)
    )
    Lifetime=&RandomRangeFloat(Minimum=0.2, Maximum=0.3)
    Mesh Scale Mode=::NewEnumerator1
    Mesh Uniform Scale Max=0.15
    Mesh Uniform Scale Min=0.1
    Sprite Size Mode=::NewEnumerator3
    Uniform Sprite Size=20
    UsePositionOffset=false
    Write Material Random=true
   )
   &SampleParticlesFromOtherEmitter(
    Applied Position Offset:Vector3f=<0 0 5>
    Mesh Scale Sampling:ENiagara_DirectReadSamplingMode=::NewEnumerator1
    Reject Particle If Sampled Index Is Invalid=false
   )
   &InitialMeshOrientation(Use Rotation Vector=true)
   /SetVariables(Particles.SpriteSize=&Multiply_Vector2D_ByFloat(
     Float=&ClampFloat(Float=&MakeFloatFromVector(Vector=$Particles.Scale, Channel=::NewEnumerator0))
     Vector2D=$Particles.SpriteSize
    )) {isSetParameters=true, module=SetVariables_#5}
  ]
  particleUpdateScript [
   &ParticleState()
   &ScaleMeshSize(Scale Factor=&RandomRangeVector(Minimum=<1 1 2.2>, Maximum=<1.3 1.3 2.2>))
   &DynamicMaterialParameters(
    Index 0 Param 2=&RandomRangeFloat(Minimum=0, Maximum=0)
    Index 0 Param 4=13
    Index 1 Param 1=&FloatFromCurve(FloatCurve=DataInterface() {propertyValues {Curve {keys [<at 0 0.5>,<at 1 1>]}}})
    Index 1 Param 2=2
    Index 1 Param 3=&FloatFromCurve(FloatCurve=DataInterface() {propertyValues {Curve {keys [{interpMode=RCIM_Cubic, tangentMode=RCTM_Break, time=0.5, value=-0.5}, <at 1 0.1>]}}})
    Index 2 Param 1=&Multiply_Float(
     A=&FloatFromCurve(FloatCurve=DataInterface() {propertyValues {Curve {keys [{interpMode=RCIM_Cubic, leaveTangent=8.1085, tangentMode=RCTM_Break, time=0.8568, value=1}, {tangentMode=RCTM_User, time=1, value=0}]}}})
     B=&Divide_Float(
      A=&MakeFloatFromVector(Vector=$Particles.Scale, Channel=::NewEnumerator0)
      B=&RandomRangeFloat(Minimum=10, Maximum=15)
     )
    )
    Index 2 Param 2=&FloatFromCurve(
     FloatCurve=DataInterface() {propertyValues {Curve {keys [
         {interpMode=RCIM_Cubic, tangentMode=RCTM_Break, time=0.7, value=0}
         {arriveTangent=-0.0859, interpMode=RCIM_Cubic, tangentMode=RCTM_Break, time=1, value=-1}
        ]}}}
     Scale Curve=50
    )
    Index 2 Param 3=&Waveform(
     Global Input Position=&RandomRangeFloat(Minimum=0, Maximum=15)
     Global Phase=$Particles.Age
     "Waveform [1]"=::NewEnumerator12
     "[1] Phase Scale"=2
     "[1] Frequency"=1
     "[1] Amplitude Min/Max"=<0 1>
    )
    Index 2 Param 4=25
    Index 3 Param 1=&FloatFromCurve(
     FloatCurve=DataInterface() {propertyValues {Curve {keys [
         {interpMode=RCIM_Cubic, leaveTangent=-0.0025, tangentMode=RCTM_Break, time=0, value=0}
         {arriveTangent=1.8528, interpMode=RCIM_Cubic, tangentMode=RCTM_Break, time=1, value=1}
        ]}}}
     Scale Curve=&RandomRangeFloat(Minimum=0, Maximum=0)
    )
    Index 3 Param 2=&FloatFromCurve(
     FloatCurve=DataInterface() {propertyValues {Curve {keys [
         {interpMode=RCIM_Cubic, leaveTangent=-0.0025, tangentMode=RCTM_Break, time=0, value=0.5}
         {arriveTangent=-0.0316, interpMode=RCIM_Cubic, tangentMode=RCTM_Break, time=1, value=-0.2}
        ]}}}
     Scale Curve=0.1
    )
    Index 3 Param 3=&FloatFromCurve(FloatCurve=DataInterface() {propertyValues {Curve {keys [
         {interpMode=RCIM_Cubic, leaveTangent=1.1034, tangentMode=RCTM_Break, time=0, value=-1}
         {arriveTangent=2.2399, tangentMode=RCTM_User, time=0.7, value=0.125}
        ]}}})
    Index 3 Param 4=&FloatFromCurve(FloatCurve=DataInterface() {propertyValues {Curve {keys [
         {interpMode=RCIM_Cubic, leaveTangent=3.0601, tangentMode=RCTM_Break, time=0.5, value=-0.5}
         {arriveTangent=3.2053, tangentMode=RCTM_User, time=1, value=1}
        ]}}})
    Module.Index0Float0Write=true
    Module.Index0Float1Write=true
    Module.Index0Float2Write=true
    Module.Index0Float3Write=true
    Module.Index1Float0Write=true
    Module.Index1Float1Write=true
    Module.Index1Float2Write=true
    Module.Index1Float3Write=true
    Module.Index2Float0Write=true
    Module.Index2Float1Write=true
    Module.Index2Float2Write=true
    Module.Index2Float3Write=true
    Module.Index3Float0Write=true
    Module.Index3Float1Write=true
    Module.Index3Float2Write=true
    Module.Index3Float3Write=true
    Write Parameter Index 1=::NewEnumerator1
    Write Parameter Index 2=::NewEnumerator1
    Write Parameter Index 3=::NewEnumerator1
   )
   &ScaleColor(
    Color Scale:LinearColor=&ColorFromCurve(ColorCurve=DataInterface() {propertyValues {
       AlphaCurve {keys [<at 0 1 RCIM_Cubic>,<at 0.7884 1>,<at 1 0 RCIM_Cubic>]}
       BlueCurve {keys [<at 0.2225 8>, {interpMode=RCIM_Cubic, leaveTangent=-19.7604, time=0.6688, value=1.5}]}
       GreenCurve {keys [<at 0.2225 8>, {interpMode=RCIM_Cubic, leaveTangent=-19.7604, time=0.6688, value=1.5}]}
       RedCurve {keys [<at 0.2225 8>, {interpMode=RCIM_Cubic, leaveTangent=-19.7604, time=0.6688, value=1.5}]}
      }})
    Scale Mode=3
   )
   &CameraOffset(Camera Offset Amount=20)
   /SetVariables(Particles.GlowColor=&ColorFromCurve(ColorCurve=DataInterface() {propertyValues {
       AlphaCurve {keys [<at 0 1 RCIM_Cubic>,<at 0.7884 1>,<at 1 0 RCIM_Cubic>]}
       BlueCurve {keys [<at 0.2225 2>, {interpMode=RCIM_Cubic, leaveTangent=-19.7604, time=0.6688, value=1}]}
       GreenCurve {keys [<at 0.2225 2>, {interpMode=RCIM_Cubic, leaveTangent=-19.7604, time=0.6688, value=1}]}
       RedCurve {keys [<at 0.2225 2>, {interpMode=RCIM_Cubic, leaveTangent=-19.7604, time=0.6688, value=1}]}
      }})) {isSetParameters=true}
  ]
  rendererClasses [&NiagaraMeshRendererProperties, &NiagaraSpriteRendererProperties]
  renderers [
   {rendererClass=&NiagaraMeshRendererProperties, rendererIndex=0}
   {rendererClass=&NiagaraMeshRendererProperties, rendererIndex=1}
   {rendererClass=&NiagaraSpriteRendererProperties, rendererIndex=2}
  ]
 }
}
summary {
 systemName=NS_CubeSplitter_CharHit
 emitters [
  {emitterName=AttachCubes, rendererClasses [&NiagaraMeshRendererProperties, &NiagaraSpriteRendererProperties]}
  {emitterName=Light, rendererClasses []}
  {emitterName=BurstRunes, rendererClasses [&NiagaraSpriteRendererProperties]}
  {emitterName=Lightning, rendererClasses [&NiagaraMeshRendererProperties, &NiagaraSpriteRendererProperties]}
  {emitterName=Decals, rendererClasses [&NiagaraSpriteRendererProperties]}
  {emitterName=LightRays, rendererClasses []}
 ]
 userVariables [
  User.SkeletalMesh {
   type=NiagaraDataInterfaceSkeletalMesh
   defaultValue {
    struct=NiagaraExt_VariableValue_DataInterface
    value {dataInterface=&self:NiagaraDataInterfaceSkeletalMesh_0}
   }
  }
  User.SpawnDensity {type=NiagaraInt32, defaultValue=4}
  User.System_SpawnDensity {type=NiagaraInt32}
 ]
}
