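# Niagara system for LoudKiwi's CubeSplitter ground-hit effect, built from mesh-particle emitters, with empty Airglow and Flare emitters.
# Cubes bursts randomly scaled, quarter-turned meshes around a small ring, offsetting them by size and feeding age curves and transformed spawn position into material parameters.
# Lightning continuously spawns short-lived purple meshes on a wider ring, animated mostly through dynamic material parameters.
# Rays bursts meshes that scale up while shifting colour over life.
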NiagaraSystem /LoudKiwi/Assets/NS_LoudKiwi_CubeSplitter_GroundHit
dependencies {
 usedDataInterfaces [
  &NiagaraDataInterfaceCurve
  &NiagaraDataInterfaceColorCurve
  &NiagaraDataInterfaceMeshRendererInfo
  &NiagaraDataInterfaceArrayFloat
 ]
 usedDynamicInputs [&MakeVectorFromLinearColorRGB]
 usedModules [&SystemState]
}
emitters {
 Airglow {rendererClasses [], renderers []}
 Cubes {
  emitterUpdateScript [
   &EmitterState()
   &SpawnBurst_Instantaneous(
    Spawn Count=&RandomRangeInt(Minimum=12, Maximum=18)
    Use Spawn Probability=false
    UseLoopCountLimit=false
   )
   /SetVariables(Emitter.Glow:NiagaraFloat=&FloatFromCurve(
     FloatCurve=DataInterface() {propertyValues {Curve {
        defaultValue=1
        keys [
         {arriveTangent=0.6183, interpMode=RCIM_Cubic, leaveTangent=0.6183, tangentMode=RCTM_User, time=0, value=0}
         {arriveTangent=3.0444, interpMode=RCIM_Cubic, leaveTangent=3.0444, tangentMode=RCTM_User, time=1.6312, value=2.2144}
         {interpMode=RCIM_Cubic, tangentMode=RCTM_User, time=3, value=4}
        ]
       }}}
     CurveIndex=&Multiply_Float(A=$Emitter.Age, B=1.3)
     Scale Curve=3
    )) {isSetParameters=true}
  ]
  particleSpawnScript [
   &InitializeParticle(
    Color Mode=::NewEnumerator0
    Mesh Scale=&RandomRangeVector(Minimum=<0.25 0.25 0.25>, Maximum=<0.5 0.5 0.5>)
    Mesh Scale Mode=::NewEnumerator4
    Position Offset=<0 0 -11>
    UsePositionOffset=true
    Write Material Random=true
   )
   &ShapeLocation(
    Disc Coverage=1
    MaskSpawnGroup=false
    Override Seed=false
    Ring Radius=15
    Rotation Axis=<0 0 1>
    Rotation Mode=::NewEnumerator1
    Shape Primitive=::NewEnumerator4
    UseOwnerScale=true
   )
   &InitialMeshOrientation(
    Mesh Orientation Mode=::NewEnumerator1
    Rotation=&MakeVector(
     X=&SelectFloatFromArray(
      Array Sampling Mode=::NewEnumerator0
      Direct Array Index=?
      Float Selection Array=DataInterface() {propertyValues {FloatData [0,0.25,0.5,0.75]}}
      Random Seed=0
     )
     Y=0
     Z=0
    )
    Use Rotation Vector=true
   )
   &InitialMeshOrientation(
    Mesh Orientation Mode=::NewEnumerator1
    Rotation=&MakeVector(
     X=0
     Y=&MakeFloatFromVector(
      Vector=&MakeVector(
       X=&SelectFloatFromArray(
        Array Sampling Mode=::NewEnumerator0
        Direct Array Index=?
        Float Selection Array=DataInterface() {propertyValues {FloatData [0,0.25,0.5,0.75]}}
        Random Seed=0
       )
       Y=0
       Z=0
      )
      Channel=::NewEnumerator0
     )
     Z=0
    )
    Use Rotation Vector=true
   ) {module=InitialMeshOrientation001}
   /SetVariables(Particles.Position=&AddVectorToPosition(
     Position=$Particles.Initial.Position
     Vector=&Multiply_VectorByFloat(
      Vector=&MakeVector(
       X=&UniformAOrBFloat(
        A=&RandomRangeFloat(Minimum=-10, Maximum=-8)
        B=&RandomRangeFloat(Minimum=10, Maximum=8)
        Evaluation Type=::NewEnumerator0
        Recalculate Random Each Loop=true
        Override Seed=false
       )
       Y=&UniformAOrBFloat(
        A=&RandomRangeFloat(Minimum=-10, Maximum=-8)
        B=&RandomRangeFloat(Minimum=10, Maximum=8)
        Evaluation Type=::NewEnumerator0
        Recalculate Random Each Loop=true
        Override Seed=false
       )
       Z=&UniformAOrBFloat(A=-1, Evaluation Type=::NewEnumerator0, Recalculate Random Each Loop=true, Override Seed=false)
      )
      Float=&RemapRange(
       Input Value To Remap=&MakeFloatFromVector(Vector=$Particles.Scale, Channel=::NewEnumerator0)
       Input Min=0
       Input Max=0.5
       Output Min=0
       Output Max=1
      )
     )
    )) {isSetParameters=true}
  ]
  particleUpdateScript [
   &ParticleState()
   &ScaleMeshSize(Scale Factor=<0.5 0.5 0.5>)
   &CameraOffset(Camera Offset Amount=&RandomRangeFloat(Minimum=3, Maximum=4))
   &DynamicMaterialParameters(
    Index 0 Param 4=&Lerp_Float(A=20, B=100, Alpha=$Particles.NormalizedAge)
    Index 0 Vector 1:Vector3f=&TransformPosition(
     Position=&Multiply_VectorByFloat(Vector=Linked() {linkedVariable {name=Particles.Initial.Position, type=NiagaraPosition}}, Float=0.5)
     Source Space=::Local
     Destination Space=::World
    )
    Index 1 Param 1=&Lerp_Float(
     A=&RandomRangeFloat(Minimum=0, Maximum=0.3)
     B=1
     Alpha=&FloatFromCurve(FloatCurve=DataInterface() {propertyValues {Curve {
         defaultValue=1
         keys [{arriveTangent=3.0442, interpMode=RCIM_Cubic, leaveTangent=3.0442, tangentMode=RCTM_User, time=0, value=0}, <at 0.5 1>]
        }}})
    )
    Index 1 Param 2=&FloatFromCurve(
     FloatCurve=DataInterface() {propertyValues {Curve {
        defaultValue=1
        keys [
         {interpMode=RCIM_Cubic, leaveTangentWeight=0.1667, tangentMode=RCTM_User, time=0.2, value=0.4}
         {arriveTangentWeight=0.1667, interpMode=RCIM_Cubic, leaveTangentWeight=0.1667, tangentMode=RCTM_User, time=0.4, value=0.15}
         {arriveTangent=1.8908, tangentMode=RCTM_User, time=0.9, value=1}
        ]
       }}}
     Scale Curve=0.5
    )
    Module.Index0Float3Write=true
    Module.Index0VectorWrite=true
    Module.Index1Float0Write=true
    Module.Index1Float1Write=true
    Module.Index1Float2Write=false
    Module.Index1Float3Write=false
    Write Parameter Index 0=::NewEnumerator3
    Write Parameter Index 1=::NewEnumerator1
   )
  ]
  rendererClasses [&NiagaraMeshRendererProperties]
  renderers [{rendererClass=&NiagaraMeshRendererProperties, rendererIndex=0}]
 }
 Flare {rendererClasses [], renderers []}
 Lightning {
  emitterUpdateScript [
   &EmitterState()
   &SpawnRate(SpawnRate=8, Use Spawn Probability=false)
  ]
  particleSpawnScript [
   &InitializeParticle(
    Color=&MultiplyLinearColor(
     Linear Color=&RandomRangeLinearColor(Minimum=<rgba 0.5306 0 1 0.3>, Maximum=<rgba 0.9352 0 1.8 1>)
     Scale Factor=&MakeVector4_FromFloat(Float=0.8)
    )
    Lifetime=&RandomRangeFloat(Minimum=0.2, Maximum=0.3)
    Mesh Scale Mode=::NewEnumerator1
    Mesh Uniform Scale Max=0.3
    Mesh Uniform Scale Min=0.15
    Sprite Size Mode=::NewEnumerator3
    Uniform Sprite Size=20
    UsePositionOffset=false
    Write Material Random=true
   )
   &ShapeLocation(
    Disc Coverage=1
    MaskSpawnGroup=false
    Override Seed=false
    Ring Radius=40
    Rotation Axis=<0 0 1>
    Rotation Mode=::NewEnumerator1
    Shape Primitive=::NewEnumerator4
    UseOwnerScale=true
   )
   &InitialMeshOrientation(Use Rotation Vector=true)
  ]
  particleUpdateScript [
   &ParticleState()
   &ScaleMeshSize(Scale Factor=&RandomRangeVector(Minimum=<0.7 0.7 1.2>, Maximum=<0.9 0.9 1.2>))
   &DynamicMaterialParameters(
    Index 0 Param 2=&RandomRangeFloat(Minimum=0, Maximum=0)
    Index 0 Param 4=13
    Index 1 Param 1=&FloatFromCurve(FloatCurve=DataInterface() {propertyValues {Curve {keys [<at 0 0.5>,<at 1 1>]}}})
    Index 1 Param 2=2
    Index 1 Param 3=&FloatFromCurve(FloatCurve=DataInterface() {propertyValues {Curve {keys [{interpMode=RCIM_Cubic, tangentMode=RCTM_Break, time=0.5, value=-0.5}, <at 1 0.1>]}}})
    Index 2 Param 1=&Multiply_Float(
     A=&FloatFromCurve(FloatCurve=DataInterface() {propertyValues {Curve {keys [{interpMode=RCIM_Cubic, leaveTangent=8.1085, tangentMode=RCTM_Break, time=0.8568, value=1}, {tangentMode=RCTM_User, time=1, value=0}]}}})
     B=&Divide_Float(
      A=&MakeFloatFromVector(Vector=$Particles.Scale, Channel=::NewEnumerator0)
      B=&RandomRangeFloat(Minimum=4, Maximum=15)
     )
    )
    Index 2 Param 2=&FloatFromCurve(
     FloatCurve=DataInterface() {propertyValues {Curve {keys [
         {interpMode=RCIM_Cubic, tangentMode=RCTM_Break, time=0.7, value=0}
         {arriveTangent=-0.0859, interpMode=RCIM_Cubic, tangentMode=RCTM_Break, time=1, value=-1}
        ]}}}
     Scale Curve=50
    )
    Index 2 Param 3=&Waveform(
     Global Input Position=&RandomRangeFloat(Minimum=0, Maximum=15)
     Global Phase=$Particles.Age
     "Waveform [1]"=::NewEnumerator12
     "[1] Phase Scale"=2
     "[1] Frequency"=1
     "[1] Amplitude Min/Max"=<0 1>
    )
    Index 2 Param 4=25
    Index 3 Param 1=&FloatFromCurve(
     FloatCurve=DataInterface() {propertyValues {Curve {keys [
         {interpMode=RCIM_Cubic, leaveTangent=-0.0025, tangentMode=RCTM_Break, time=0, value=0}
         {arriveTangent=1.8528, interpMode=RCIM_Cubic, tangentMode=RCTM_Break, time=1, value=1}
        ]}}}
     Scale Curve=&RandomRangeFloat(Minimum=0, Maximum=0)
    )
    Index 3 Param 2=&FloatFromCurve(
     FloatCurve=DataInterface() {propertyValues {Curve {keys [
         {interpMode=RCIM_Cubic, leaveTangent=-0.0025, tangentMode=RCTM_Break, time=0, value=0.5}
         {arriveTangent=-0.0316, interpMode=RCIM_Cubic, tangentMode=RCTM_Break, time=1, value=-0.2}
        ]}}}
     Scale Curve=0.1
    )
    Index 3 Param 3=&FloatFromCurve(FloatCurve=DataInterface() {propertyValues {Curve {keys [
         {interpMode=RCIM_Cubic, leaveTangent=1.1034, tangentMode=RCTM_Break, time=0, value=-1}
         {arriveTangent=2.2399, tangentMode=RCTM_User, time=0.7, value=0.125}
        ]}}})
    Index 3 Param 4=&FloatFromCurve(FloatCurve=DataInterface() {propertyValues {Curve {keys [
         {interpMode=RCIM_Cubic, leaveTangent=3.0601, tangentMode=RCTM_Break, time=0.5, value=-0.5}
         {arriveTangent=3.2053, tangentMode=RCTM_User, time=1, value=1}
        ]}}})
    Module.Index0Float0Write=true
    Module.Index0Float1Write=true
    Module.Index0Float2Write=true
    Module.Index0Float3Write=true
    Module.Index1Float0Write=true
    Module.Index1Float1Write=true
    Module.Index1Float2Write=true
    Module.Index1Float3Write=true
    Module.Index2Float0Write=true
    Module.Index2Float1Write=true
    Module.Index2Float2Write=true
    Module.Index2Float3Write=true
    Module.Index3Float0Write=true
    Module.Index3Float1Write=true
    Module.Index3Float2Write=true
    Module.Index3Float3Write=true
    Write Parameter Index 1=::NewEnumerator1
    Write Parameter Index 2=::NewEnumerator1
    Write Parameter Index 3=::NewEnumerator1
   )
   &ScaleColor(
    Color Scale:LinearColor=&ColorFromCurve(ColorCurve=DataInterface() {propertyValues {
       AlphaCurve {keys [<at 0 1 RCIM_Cubic>,<at 0.7884 1>,<at 1 0 RCIM_Cubic>]}
       BlueCurve {keys [<at 0.2225 8>, {interpMode=RCIM_Cubic, leaveTangent=-19.7604, time=0.6688, value=1.5}]}
       GreenCurve {keys [<at 0.2225 8>, {interpMode=RCIM_Cubic, leaveTangent=-19.7604, time=0.6688, value=1.5}]}
       RedCurve {keys [<at 0.2225 8>, {interpMode=RCIM_Cubic, leaveTangent=-19.7604, time=0.6688, value=1.5}]}
      }})
    Scale Mode=3
   )
   &CameraOffset(Camera Offset Amount=20)
  ]
  rendererClasses [&NiagaraMeshRendererProperties]
  renderers [
   {rendererClass=&NiagaraMeshRendererProperties, rendererIndex=0}
   {rendererClass=&NiagaraMeshRendererProperties, rendererIndex=1}
  ]
 }
 Rays {
  emitterUpdateScript [
   &EmitterState()
   &SpawnBurst_Instantaneous(Spawn Count=16, Use Spawn Probability=false, UseLoopCountLimit=false)
  ]
  particleSpawnScript [
   &InitializeParticle(
    Color=<rgba 1.9037 0 4 0.5>
    Mesh Scale Mode=::NewEnumerator3
    Mesh Uniform Scale=0.7
    Position Offset=<0 0 -12>
    UsePositionOffset=true
    Write Material Random=true
   )
   &InitialMeshOrientation(Use Rotation Vector=false)
  ]
  particleUpdateScript [
   &ParticleState()
   &ScaleMeshSize(Scale Factor=&Multiply_VectorByFloat(
     Vector=$Particles.Initial.Scale
     Float=&FloatFromCurve(
      FloatCurve=DataInterface() {propertyValues {Curve {
         defaultValue=1
         keys [
          <at 0 0.2 RCIM_Cubic>
          {interpMode=RCIM_Cubic, tangentMode=RCTM_User, time=0.5, value=1}
          {interpMode=RCIM_Cubic, leaveTangent=-2.2728, time=1, value=1}
         ]
        }}}
      Scale Curve=1.3
     )
    ))
   &CameraOffset(Camera Offset Amount=10)
   &ScaleColor(
    Linear Color Curve=DataInterface() {propertyValues {
      AlphaCurve {keys [<at 0.1525 0.3>,<at 1 0.2>]}
      BlueCurve {keys [<at 0.006 0>,<at 0.3863 0.6823>,<at 1 2.4384>]}
      GreenCurve {keys [<at 0.006 0>,<at 0.3863 0>,<at 1 0.2958>]}
      RedCurve {keys [<at 0.006 0>,<at 0.3863 0.5216>,<at 1 3>]}
     }}
    Scale Mode=2
   )
  ]
  rendererClasses [&NiagaraMeshRendererProperties]
  renderers [{rendererClass=&NiagaraMeshRendererProperties, rendererIndex=0}]
 }
}
summary {
 systemName=NS_LoudKiwi_CubeSplitter_GroundHit
 emitters [
  {emitterName=Rays, rendererClasses [&NiagaraMeshRendererProperties]}
  {emitterName=Lightning, rendererClasses [&NiagaraMeshRendererProperties]}
  {emitterName=Cubes, rendererClasses [&NiagaraMeshRendererProperties]}
  {emitterName=Airglow, rendererClasses []}
  {emitterName=Flare, rendererClasses []}
 ]
 userVariables []
}
