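# Niagara system for a short violet burst, named as a status-apply effect, that plays over a character's body.
# Decals bursts camera-offset sprites that sample filtered spine and limb bones by particle ID each update, jittered by a cylinder shape.
# Dynamic material parameters and fading color drive those sprites, with brightness also selected by System.SpawnDensity.
# Rays bursts briefer mesh particles that shrink over life with a color curve; Flare has no renderer.
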NiagaraSystem /LoudKiwi/Assets/NS_LoudKiwi_StatusApply
dependencies {
 usedDataInterfaces [
  &NiagaraDataInterfaceCurve
  &NiagaraDataInterfaceColorCurve
  &NiagaraDataInterfaceArrayInt32
  &NiagaraDataInterfaceSkeletalMesh
  &NiagaraDataInterfaceArrayFloat
 ]
 usedDynamicInputs [&Multiply_VectorByFloat]
 usedModules [&SystemState]
}
emitters {
 Decals {
  emitterUpdateScript [
   &EmitterState()
   &SpawnBurst_Instantaneous(Spawn Count=10, Use Spawn Probability=false, UseLoopCountLimit=false)
  ]
  particleSpawnScript [
   /SetVariables(Particles.Index:NiagaraInt32=&SelectIntFromArray(
     Direct Array Index=&ReturnParticleIDAttributes()
     Int Selection Array=DataInterface() {propertyValues {IntData [0,1,2,8,9,17,18]}}
    )) {isSetParameters=true}
   &InitializeParticle(
    Color=<rgba 9.1665 0 20 1>
    Lifetime=0.5
    Sprite Size Mode=::NewEnumerator3
    Uniform Sprite Size=50
    UsePositionOffset=false
    Write Material Random=true
   )
   &CameraOffset(Camera Offset Amount=32)
  ]
  particleUpdateScript [
   &ParticleState()
   &SampleSkeletalMesh(
    AdditionalYawPitchRoll=false
    Bone / Socket Index:NiagaraInt32=&ModuloInt(A=&ReturnParticleIDAttributes(), B=11)
    Bone Sampling Mode:ENiagaraBoneSamplingMode=::NewEnumerator4
    Skeletal Mesh=DataInterface() {propertyValues {
      DefaultMesh=null
      ExcludeBoneName=root
      OptionalComponentClass=null
      PreviewMesh=/Game/Characters/Player/Female/Medium/Bodies/F_MED_Ramirez_Fallback/Meshes/F_MED_Ramirez_Fallback
      SkinningMode=SkinOnTheFly
      SoftSourceActor=null
      SourceMode=Default
      UvSetIndex=0
      WholeMeshLOD=-1
      bExcludeBone=false
      bReadDeformedGeometry=true
      bRequireCurrentFrameData=true
      ComponentTags []
      FilteredBones [spine_01,spine_03,spine_05,upperarm_l,lowerarm_l,upperarm_r,lowerarm_r,thigh_l,calf_l,thigh_r,calf_r]
      FilteredSockets []
      MeshUserParameter {parameter {name=None, typeDefHandle {registeredTypeIndex=1}, varData []}}
      SamplingRegions []
     }}
   )
   &ShapeLocation(
    Cylinder Height=15
    Cylinder Radius=5
    MaskSpawnGroup=false
    Override Seed=false
    Shape Origin=$Particles.SampleSkeletalMesh.SampledPosition
    Shape Primitive=::NewEnumerator1
    Surface Only=false
    UseOwnerScale=true
   )
   &DynamicMaterialParameters(
    Index 0 Param 1=4
    Index 0 Param 2=1
    Index 0 Param 3=&Multiply_Float(A=$Particles.Age, B=0.1)
    Index 0 Param 4=&RandomRangeFloat(Minimum=0.75, Maximum=0.9)
    Index 1 Param 1=250
    Index 1 Param 2=1
    Module.Param0WriteEnabled=true
    Module.Param0WriteEnabled1=true
    Module.Param1WriteEnabled=true
    Module.Param1WriteEnabled1=true
    Module.Param2WriteEnabled=true
    Module.Param2WriteEnabled1=false
    Module.Param3WriteEnabled=true
    Module.Param3WriteEnabled1=false
    Write Parameter Index 1=true
   )
   &ScaleColor(
    Scale RGB=&VectorFromFloat(Value=&FloatFromCurve(FloatCurve=DataInterface() {propertyValues {Curve {keys [{arriveTangent=-8.2798, interpMode=RCIM_Cubic, leaveTangent=-8.2798, tangentMode=RCTM_User, time=0, value=3}, <at 1 0>]}}}))
    ScaleA=true
    ScaleRGB=true
   )
   &ScaleColor(
    Curve Index=&SelectFloatFromArray(Direct Array Index=$System.SpawnDensity, Float Selection Array=DataInterface() {propertyValues {FloatData [0,0.25,0.35,0.5,0.75]}})
    Linear Color Curve=DataInterface() {propertyValues {
      AlphaCurve {keys [<at 0 1>]}
      BlueCurve {keys [<at 0 0.5>,<at 0.2529 0.75>,<at 0.5372 1>,<at 0.7754 1.25>,<at 1 1.5>]}
      GreenCurve {keys [<at 0 0.5>,<at 0.2529 0.75>,<at 0.5372 1>,<at 0.7754 1.25>,<at 1 1.5>]}
      RedCurve {keys [<at 0 0.5>,<at 0.2529 0.75>,<at 0.5372 1>,<at 0.7754 1.25>,<at 1 1.5>]}
     }}
    Scale Mode=2
   ) {module=ScaleColor001}
  ]
  rendererClasses [&NiagaraSpriteRendererProperties]
  renderers [{rendererClass=&NiagaraSpriteRendererProperties}]
 }
 Flare {rendererClasses [], renderers []}
 Rays {
  emitterUpdateScript [
   &EmitterState()
   &SpawnBurst_Instantaneous(Spawn Count=8, Use Spawn Probability=false, UseLoopCountLimit=false)
  ]
  particleSpawnScript [
   &InitializeParticle(
    Color=<rgba 5.7112 0 12 1>
    Lifetime=0.25
    Mesh Scale Mode=::NewEnumerator3
    Mesh Uniform Scale=2
    UsePositionOffset=true
    Write Material Random=true
   )
   &InitialMeshOrientation(Use Rotation Vector=false)
  ]
  particleUpdateScript [
   &ParticleState()
   &ScaleMeshSize(Scale Factor=&VectorFromFloat(Value=&FloatFromCurve(FloatCurve=DataInterface() {propertyValues {Curve {keys [{arriveTangent=-2.7237, interpMode=RCIM_Cubic, leaveTangent=-2.7237, tangentMode=RCTM_User, time=0, value=1}, <at 1 0>]}}})))
   &CameraOffset(Camera Offset Amount=10)
   &ScaleColor(
    Linear Color Curve=DataInterface() {propertyValues {
      AlphaCurve {keys [<at 0.1525 0.3>,<at 1 0.2>]}
      BlueCurve {keys [<at 0.006 0>,<at 0.3863 0.6823>,<at 1 2.4384>]}
      GreenCurve {keys [<at 0.006 0>,<at 0.3863 0>,<at 1 0.2958>]}
      RedCurve {keys [<at 0.006 0>,<at 0.3863 0.5216>,<at 1 3>]}
     }}
    Scale Mode=2
   )
  ]
  rendererClasses [&NiagaraMeshRendererProperties]
  renderers [{rendererClass=&NiagaraMeshRendererProperties, rendererIndex=0}]
 }
}
summary {
 systemName=NS_LoudKiwi_StatusApply
 emitters [
  {emitterName=Rays, rendererClasses [&NiagaraMeshRendererProperties]}
  {emitterName=Flare, rendererClasses []}
  {emitterName=Decals, rendererClasses [&NiagaraSpriteRendererProperties]}
 ]
 userVariables []
}
