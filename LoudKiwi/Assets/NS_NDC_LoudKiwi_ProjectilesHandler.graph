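# Niagara handler system that renders LoudKiwi projectile tracers and smoke trails from Data Channel reads.
# Both emitters spawn from the channel, sizing purple effects by ammo type and LOD distance.
# Bullet002 draws velocity-stretched sprites, flags hitscan shots, and animates end-segment position and length with curves, offsetting sprites backward along their direction.
# Trail002 draws fading, widening ribbons with curve-driven smoke material parameters.
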NiagaraSystem /LoudKiwi/Assets/NS_NDC_LoudKiwi_ProjectilesHandler
dependencies {
 usedDataInterfaces [
  &NiagaraDataInterfaceCurve
  &NiagaraDataInterfaceArrayFloat
  /Script/Niagara.NiagaraDataInterfaceDataChannelRead
 ]
 usedDynamicInputs []
 usedModules [
  &SystemState
  /Niagara/Modules/System/CompleteIfUnused
  &self:SystemSpawnScript.SystemScriptSource.SystemScriptGraph.NiagaraNodeAssignment_0.SetVariables_#1
 ]
}
emitters {
 Bullet002 {
  emitterUpdateScript [
   &EmitterState()
   /Game/Effects/Niagara/DataChannels/Projectiles/Spawn_NDCFortProjectiles(Reader:NiagaraDataInterfaceDataChannelRead=$System.NDCReader)
  ]
  particleSpawnScript [
   /Game/Effects/Niagara/DataChannels/Projectiles/InitializeParticleFromNDC_FortProjectiles(
    End Lifetime=true
    End Lifetime Max=0.1167
    End Lifetime Min=0.05
    FallbackAmmoType=2
    FallbackOrigin:NiagaraPosition=<0 50 0>
    Intro Time=true
    LOD Output Scale Min=0.5
    Length Scale Array=DataInterface() {propertyValues {GpuSyncMode=SyncCpuToGpu, MaxElements=4, FloatData [0.75,0.85,1,0.7]}}
    Length Scale by Array=true
    Length by Velocity=true
    Lifetime=0.0333
    Scale Array=DataInterface() {propertyValues {GpuSyncMode=SyncCpuToGpu, MaxElements=4, FloatData [8,12,16,6]}}
    Scale by Array=true
    Use LOD Distance Scaling=true
   )
   &KillParticles(Kill Particles=$Particles.IsStartPosition) {module=KillParticles001}
   &InitializeParticle(
    Color Mode=::NewEnumerator0
    Lifetime=$Particles.Lifetime
    Position=$Particles.CurrentPosition
    Position Mode=::NewEnumerator1
    Sprite Size=&MakeVector2D(
     X=$Particles.ProjectileScale
     Y=&Multiply_Float(A=$Particles.ProjectileLength, B=$Particles.NormalizedIntroDuration)
    )
    Sprite Size Mode=::NewEnumerator4
    UsePositionOffset=false
    Write Material Random=true
   )
   &DynamicMaterialParameters(
    Index 0 Param 1=2
    Index 0 Param 2=1
    Index 0 Param 3=3
    Index 0 Param 4=0.25
    Module.Index0Param0Write=true
    Module.Index0Param1Write=true
    Module.Index0Param2Write=true
    Module.Index0Param3Write=true
   )
   &self:CustomDataModule() {module=CustomDataModule}
   /SetVariables(Particles.IsHitScan:NiagaraBool=&Boolean_AndOperation(
     Bool1=$Particles.IsEndPosition
     Bool2=&SetBoolByFloatComparison(
      A=&VectorLength(
       Vector=&SubtractVectorFromPosition(Position=$Particles.PreviousPosition, Vector=Linked() {linkedVariable {name=Particles.OriginPosition, type=NiagaraPosition}})
       Safe Length Calculation=false
      )
      B=0
      Comparison Type=::NewEnumerator4
     )
    )) {isSetParameters=true}
   /SetVariables(Particles.Color=<rgba 0.5597 0 1 1>) {isSetParameters=true}
  ]
  particleUpdateScript [
   &ParticleState()
   &KillParticles(Kill Particles=&Boolean_NotOperation(Bool=$Particles.IsEndPosition))
   /SetVariables(Particles.HitScanAge:NiagaraFloat=&MakeCustomFloatFromBool(
     Bool=$Particles.IsHitScan
     True Float=$Particles.NormalizedAge
     False Float=&ClampFloat(Float=&Add_Float(A=&Multiply_Float(A=$Particles.NormalizedAge, B=0.5), B=0.66))
    )) {isSetParameters=true}
   /SetVariables(Particles.Position=$Particles.Initial.Position) {isSetParameters=true}
   /SetVariables(Particles.Position=&MakeCustomVectorFromBool(
     Bool=$Particles.IsEndPosition
     True Vector=&Lerp_Vector(
      A=Linked() {linkedVariable {name=Particles.PreviousPosition, type=NiagaraPosition}}
      B=Linked() {linkedVariable {name=Particles.Position, type=NiagaraPosition}}
      Alpha=&FloatFromCurve(
       FloatCurve=DataInterface() {propertyValues {Curve {keys [{arriveTangent=2.1933, time=0, value=0}, <at 0.66 1>]}}}
       CurveIndex=$Particles.HitScanAge
      )
     )
     False Vector=Linked() {linkedVariable {name=Particles.Position, type=NiagaraPosition}}
    )) {isSetParameters=true}
   /SetVariables(Particles.Tracer Scale:NiagaraFloat=&ClampFloat(Float=&Divide_Float(
      A=&VectorLength(
       Vector=&SubtractVectorFromPosition(Position=$Particles.Position, Vector=Linked() {linkedVariable {name=Particles.OriginPosition, type=NiagaraPosition}})
       Safe Length Calculation=false
      )
      B=$Particles.ProjectileLength
     ))) {isSetParameters=true}
   &ScaleSpriteSize(
    Scale Factor=&MakeVector2D(
     X=1
     Y=&FloatFromCurve(
      FloatCurve=DataInterface() {propertyValues {Curve {keys [<at 0 0>, <at 0.33 1>, {arriveTangent=-4.3415, time=0.66, value=1}, {arriveTangent=-1.2985, leaveTangent=-1.2985, time=1, value=0}]}}}
      CurveIndex=$Particles.HitScanAge
      Scale Curve=$Particles.Tracer Scale
     )
    )
    Scale Sprite Size Mode=::NewEnumerator2
   ) {module=ScaleSpriteSize001}
   /SetVariables(Particles.Muzzle Position Offset:Vector3f=&Multiply_VectorByFloat(
     Vector=$Particles.Direction
     Float=&Multiply_Float(A=&MakeFloatFromVector2d(Vector 2D=$Particles.SpriteSize, Channel=::NewEnumerator1), B=-0.49)
    )) {isSetParameters=true}
   /SetVariables(Particles.Position=&AddVectorToPosition(Position=$Particles.Position, Vector=$Particles.Muzzle Position Offset)) {isSetParameters=true}
  ]
  rendererClasses [&NiagaraSpriteRendererProperties]
  renderers [{rendererClass=&NiagaraSpriteRendererProperties}]
 }
 Trail002 {
  emitterUpdateScript [
   &EmitterState()
   &Spawn_NDCFortProjectiles(Reader:NiagaraDataInterfaceDataChannelRead=$System.NDCReader)
  ]
  particleSpawnScript [
   &InitializeParticleFromNDC_FortProjectiles(
    FallbackAmmoType=2
    LOD Output Scale Max=2
    Scale Array=DataInterface() {propertyValues {GpuSyncMode=SyncCpuToGpu, MaxElements=4, FloatData [15,20,24,14]}}
    Scale by Array=true
    Use LOD Distance Scaling=true
   )
   &InitializeParticle(
    Color=<rgba 0.5205 0 1.5 1>
    Lifetime=&Multiply_Float(
     A=1
     B=&SelectFloatFromArray(Direct Array Index=$Particles.AmmoType, Float Selection Array=DataInterface() {propertyValues {FloatData [0.75,1,2,0.7]}})
    )
    Mesh Scale Mode=::NewEnumerator3
    Mesh Uniform Scale=10
    Position=$Particles.CurrentPosition
    Position Mode=::NewEnumerator1
    Ribbon Width=&Multiply_Float(A=$Particles.ProjectileScale, B=&Lerp_Float(A=0.75, B=1, Alpha=$Particles.Random))
    Ribbon Width Mode=::NewEnumerator1
    UsePositionOffset=false
    Write Material Random=true
   )
   /SetVariables(Particles.Smoke WPO Scale:NiagaraFloat=&SelectFloatFromArray(Direct Array Index=$Particles.AmmoType, Float Selection Array=DataInterface() {propertyValues {FloatData [0.75,1,1.2,0.75]}})) {isSetParameters=true}
  ]
  particleUpdateScript [
   &ParticleState()
   &ScaleColor(
    Scale Alpha=&FloatFromCurve(FloatCurve=DataInterface() {propertyValues {Curve {keys [{arriveTangent=-2.5785, interpMode=RCIM_Cubic, time=0, value=1}, {interpMode=RCIM_Cubic, tangentMode=RCTM_User, time=1, value=0}]}}})
    Scale RGB=&Multiply_VectorByFloat(
     Vector=<1 1 1>
     Float=&FloatFromCurve(
      FloatCurve=DataInterface() {propertyValues {Curve {keys [
          {arriveTangent=-13.4341, interpMode=RCIM_Cubic, time=0, value=1}
          {arriveTangent=-18.2371, interpMode=RCIM_Cubic, leaveTangent=-18.2371, tangentMode=RCTM_User, time=0.2, value=0}
         ]}}}
      CurveIndex=$Particles.Age
     )
    )
    ScaleA=true
    ScaleRGB=true
   )
   &ScaleRibbonWidth(Ribbon Width Scale=&FloatFromCurve(FloatCurve=DataInterface() {propertyValues {Curve {keys [<at 0 0.4 RCIM_Cubic>, {tangentMode=RCTM_User, time=1, value=1}]}}}))
   &DynamicMaterialParameters(
    Index 0 Param 1=&FloatFromCurve(
     FloatCurve=DataInterface() {propertyValues {Curve {keys [<at 0.25 0 RCIM_Cubic>, {interpMode=RCIM_Cubic, tangentMode=RCTM_User, time=1, value=1}]}}}
     Scale Curve=&Multiply_Float(A=16, B=$Particles.Smoke WPO Scale)
    )
    Index 0 Param 2=&FloatFromCurve(
     FloatCurve=DataInterface() {propertyValues {Curve {keys [
         {arriveTangent=4.1785, interpMode=RCIM_Cubic, leaveTangent=4.1785, tangentMode=RCTM_User, time=0, value=0}
         {arriveTangent=1.1179, interpMode=RCIM_Cubic, leaveTangent=1.1179, tangentMode=RCTM_User, time=0.2042, value=0.6228}
         {arriveTangent=0.0224, interpMode=RCIM_Cubic, leaveTangent=0.0224, tangentMode=RCTM_User, time=1, value=1}
        ]}}}
     Scale Curve=-2.5
    )
    Index 0 Param 3=$Particles.Random
    Index 0 Param 4=&FloatFromCurve(
     FloatCurve=DataInterface() {propertyValues {Curve {keys [{interpMode=RCIM_Cubic, tangentMode=RCTM_User, time=0, value=0}, {interpMode=RCIM_Cubic, tangentMode=RCTM_User, time=0.33, value=1}]}}}
     Scale Curve=0.8
    )
    Module.Index0Float0Write=true
    Module.Index0Float1Write=true
    Module.Index0Float2Write=true
    Module.Index0Float3Write=true
   )
  ]
  rendererClasses [&NiagaraRibbonRendererProperties]
  renderers [{rendererClass=&NiagaraRibbonRendererProperties}]
 }
}
summary {
 systemName=NS_NDC_LoudKiwi_ProjectilesHandler
 emitters [
  {emitterName=Trail002, rendererClasses [&NiagaraRibbonRendererProperties]}
  {emitterName=Bullet002, rendererClasses [&NiagaraSpriteRendererProperties]}
 ]
 userVariables [User.SpawnTracer {type=NiagaraBool, defaultValue=true}]
}
